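# Control Rig for the female Street Goth Summer body that drives corrective offsets on skirt bones from thigh poses.
# Four chained spherical pose readers, two per thigh, weight Modify Transforms nodes that rotate and shift the same-side front and side skirt base bones.
# Three readers measure against the pelvis; the second right-thigh reader names no parent bone.
# A separate pose-reader chain, mostly with zero offsets, is disconnected from execution.
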ControlRigBlueprint /Game/Characters/Player/Female/Medium/Bodies/F_MED_StreetGothSummer/Meshes/F_MED_Street_Goth_Summer_CtrlRig
graphs [
 RigVMFunctionLibrary {links [], nodes []}
 RigVMModel {nodes [
   RigUnit_BeginExecution = &RigUnit_BeginExecution()
   n1 = &RigUnit_SphericalPoseReader(
    ActiveRegionScaleFactors=(PositiveWidth=3.5,NegativeWidth=1.5,PositiveHeight=3,NegativeHeight=3)
    ActiveRegionSize=0.1
    Debug=(bDrawDebug=false,DebugScale=25,DebugSegments=20,DebugThickness=0.25)
    DriverAxis=(X=-1,Y=0,Z=0)
    DriverItem=(Type=Bone,Name="thigh_l")
    FalloffRegionScaleFactors=(PositiveWidth=1,NegativeWidth=0.9,PositiveHeight=1,NegativeHeight=1)
    FalloffSize=0.45
    OptionalParentItem=(Type=Bone,Name="pelvis")
    RotationOffset=(X=120,Y=0,Z=0)
    ExecutePin=@RigUnit_BeginExecution.ExecutePin
   )
   > &RigUnit_ModifyTransforms(
    ItemToModify=((Item=(Type=Bone,Name="dyn_skirt_fr_l_base"),Transform=(Rotation=(X=0,Y=0,Z=0,W=1),Translation=(X=0,Y=0,Z=15),Scale3D=(X=1,Y=1,Z=1))))
    ItemToModify.0.Transform.Rotation=&RigVMFunction_MathQuaternionFromEuler(Euler=(X=0,Y=-75,Z=0)).Result
    Weight=@n1.OutputParam
   )
   > &RigUnit_ModifyTransforms(
    ItemToModify=((Item=(Type=Bone,Name="dyn_skirt_side_l_base"),Transform=(Rotation=(X=0,Y=0,Z=0,W=1),Translation=(X=-1,Y=-5,Z=1),Scale3D=(X=1,Y=1,Z=1))))
    ItemToModify.0.Transform.Rotation=&RigVMFunction_MathQuaternionFromEuler(Euler=(X=65,Y=-20,Z=0)).Result
    Weight=@n1.OutputParam
   )
   > n2 = &RigUnit_SphericalPoseReader(
    ActiveRegionScaleFactors=(PositiveWidth=3.5,NegativeWidth=1.5,PositiveHeight=3,NegativeHeight=3)
    ActiveRegionSize=0.1
    Debug=(bDrawDebug=false,DebugScale=25,DebugSegments=20,DebugThickness=0.25)
    DriverAxis=(X=1,Y=0,Z=0)
    DriverItem=(Type=Bone,Name="thigh_r")
    FalloffRegionScaleFactors=(PositiveWidth=1,NegativeWidth=1,PositiveHeight=1,NegativeHeight=1)
    FalloffSize=0.45
    OptionalParentItem=(Type=Bone,Name="pelvis")
    RotationOffset=(X=-120,Y=180,Z=0)
   )
   > &RigUnit_ModifyTransforms(
    ItemToModify=((Item=(Type=Bone,Name="dyn_skirt_fr_r_base"),Transform=(Rotation=(X=0,Y=0,Z=0,W=1),Translation=(X=0,Y=-2,Z=12),Scale3D=(X=1,Y=1,Z=1))))
    ItemToModify.0.Transform.Rotation=&RigVMFunction_MathQuaternionFromEuler(Euler=(X=-15,Y=-50,Z=0)).Result
    Weight=@n2.OutputParam
   )
   > &RigUnit_ModifyTransforms(
    ItemToModify=((Item=(Type=Bone,Name="dyn_skirt_side_r_base"),Transform=(Rotation=(X=0,Y=0,Z=0,W=1),Translation=(X=1,Y=-2,Z=1),Scale3D=(X=1,Y=1,Z=1))))
    ItemToModify.0.Transform.Rotation=&RigVMFunction_MathQuaternionFromEuler(Euler=(X=-65,Y=-25,Z=0)).Result
    Weight=@n2.OutputParam
   )
   > n3 = &RigUnit_SphericalPoseReader(
    ActiveRegionScaleFactors=(PositiveWidth=3.5,NegativeWidth=3.7,PositiveHeight=3.5,NegativeHeight=3.5)
    ActiveRegionSize=0.1
    Debug=(bDrawDebug=false,DebugScale=25,DebugSegments=20,DebugThickness=0.25)
    DriverAxis=(X=-1,Y=0,Z=0)
    DriverItem=(Type=Bone,Name="thigh_l")
    FalloffRegionScaleFactors=(PositiveWidth=1,NegativeWidth=1.2,PositiveHeight=1,NegativeHeight=1)
    FalloffSize=0.45
    OptionalParentItem=(Type=Bone,Name="pelvis")
    RotationOffset=(X=180,Y=0,Z=0)
   )
   > &RigUnit_ModifyTransforms(
    ItemToModify=((Item=(Type=Bone,Name="dyn_skirt_fr_l_base"),Transform=(Rotation=(X=0,Y=0,Z=0,W=1),Translation=(X=0,Y=0,Z=0),Scale3D=(X=1,Y=1,Z=1))))
    ItemToModify.0.Transform.Rotation=&RigVMFunction_MathQuaternionFromEuler(Euler=(X=0,Y=-10,Z=0)).Result
    Weight=@n3.OutputParam
   )
   > &RigUnit_ModifyTransforms(
    ItemToModify=((Item=(Type=Bone,Name="dyn_skirt_side_l_base"),Transform=(Rotation=(X=0,Y=0,Z=0,W=1),Translation=(X=0,Y=0,Z=0),Scale3D=(X=1,Y=1,Z=1))))
    ItemToModify.0.Transform.Rotation=&RigVMFunction_MathQuaternionFromEuler(Euler=(X=0,Y=-15,Z=0)).Result
    Weight=@n3.OutputParam
   )
   > n4 = &RigUnit_SphericalPoseReader(
    ActiveRegionScaleFactors=(PositiveWidth=3.5,NegativeWidth=3.75,PositiveHeight=3.5,NegativeHeight=3.5)
    ActiveRegionSize=0.1
    Debug=(bDrawDebug=false,DebugScale=25,DebugSegments=20,DebugThickness=0.25)
    DriverAxis=(X=1,Y=0,Z=0)
    DriverItem=(Type=Bone,Name="thigh_r")
    FalloffRegionScaleFactors=(PositiveWidth=1,NegativeWidth=1.25,PositiveHeight=1,NegativeHeight=1)
    FalloffSize=0.45
    OptionalParentItem=(Type=Bone)
    RotationOffset=(X=-180,Y=180,Z=0)
   )
   > &RigUnit_ModifyTransforms(
    ItemToModify=((Item=(Type=Bone,Name="dyn_skirt_fr_r_base"),Transform=(Rotation=(X=0,Y=0,Z=0,W=1),Translation=(X=0,Y=0,Z=0),Scale3D=(X=1,Y=1,Z=1))))
    ItemToModify.0.Transform.Rotation=&RigVMFunction_MathQuaternionFromEuler(Euler=(X=-10,Y=-20,Z=-60)).Result
    Weight=@n4.OutputParam
   )
   > &RigUnit_ModifyTransforms(
    ItemToModify=((Item=(Type=Bone,Name="dyn_skirt_side_r_base"),Transform=(Rotation=(X=0,Y=0,Z=0,W=1),Translation=(X=0,Y=0,Z=2),Scale3D=(X=1,Y=1,Z=1))))
    ItemToModify.0.Transform.Rotation=&RigVMFunction_MathQuaternionFromEuler(Euler=(X=0,Y=-40,Z=0)).Result
    Weight=@n4.OutputParam
   )
   n5 = &RigUnit_SphericalPoseReader(
    ActiveRegionScaleFactors=(PositiveWidth=3.5,NegativeWidth=1,PositiveHeight=3,NegativeHeight=3)
    ActiveRegionSize=0.1
    Debug=(bDrawDebug=false,DebugScale=25,DebugSegments=20,DebugThickness=0.25)
    DriverAxis=(X=-1,Y=0,Z=0)
    DriverItem=(Type=Bone,Name="thigh_l")
    ExecutePin=()
    FalloffRegionScaleFactors=(PositiveWidth=0.5,NegativeWidth=0.1,PositiveHeight=0.5,NegativeHeight=0.5)
    FalloffSize=0.4
    OptionalParentItem=(Type=Bone,Name="pelvis")
    RotationOffset=(X=120,Y=0,Z=0)
   )
   > &RigUnit_ModifyTransforms(
    ItemToModify=((Item=(Type=Bone,Name="dyn_skirt_fr_l_base"),Transform=(Rotation=(X=0,Y=0,Z=0,W=1),Translation=(X=0,Y=0,Z=0),Scale3D=(X=1,Y=1,Z=1))))
    ItemToModify.0.Transform.Rotation=&RigVMFunction_MathQuaternionFromEuler(Euler=(X=0,Y=0,Z=0)).Result
    Weight=@n5.OutputParam
   )
   > &RigUnit_ModifyTransforms(
    ItemToModify=((Item=(Type=Bone,Name="dyn_skirt_fr_r_base"),Transform=(Rotation=(X=0,Y=0,Z=0,W=1),Translation=(X=0,Y=0,Z=0),Scale3D=(X=1,Y=1,Z=1))))
    ItemToModify.0.Transform.Rotation=&RigVMFunction_MathQuaternionFromEuler(Euler=(X=0,Y=0,Z=0)).Result
    Weight=@n5.OutputParam
   )
   > n6 = &RigUnit_SphericalPoseReader(
    ActiveRegionScaleFactors=(PositiveWidth=3.5,NegativeWidth=1,PositiveHeight=3,NegativeHeight=3)
    ActiveRegionSize=0.1
    Debug=(bDrawDebug=false,DebugScale=25,DebugSegments=20,DebugThickness=0.25)
    DriverAxis=(X=1,Y=0,Z=0)
    DriverItem=(Type=Bone,Name="thigh_r")
    FalloffRegionScaleFactors=(PositiveWidth=0.5,NegativeWidth=0.1,PositiveHeight=0.5,NegativeHeight=0.5)
    FalloffSize=0.4
    OptionalParentItem=(Type=Bone,Name="pelvis")
    RotationOffset=(X=-120,Y=180,Z=-8)
   )
   > &RigUnit_ModifyTransforms(
    ItemToModify=((Item=(Type=Bone,Name="dyn_skirt_fr_r_base"),Transform=(Rotation=(X=0,Y=0,Z=0,W=1),Translation=(X=0,Y=0,Z=0),Scale3D=(X=1,Y=1,Z=1))))
    ItemToModify.0.Transform.Rotation=&RigVMFunction_MathQuaternionFromEuler(Euler=(X=0,Y=0,Z=0)).Result
    Weight=@n6.OutputParam
   )
   > n7 = &RigUnit_SphericalPoseReader(
    ActiveRegionScaleFactors=(PositiveWidth=3.5,NegativeWidth=3.5,PositiveHeight=3.5,NegativeHeight=3.5)
    ActiveRegionSize=0.1
    Debug=(bDrawDebug=false,DebugScale=25,DebugSegments=20,DebugThickness=0.25)
    DriverAxis=(X=-1,Y=0,Z=0)
    DriverItem=(Type=Bone,Name="thigh_l")
    FalloffRegionScaleFactors=(PositiveWidth=1,NegativeWidth=1,PositiveHeight=1,NegativeHeight=1)
    FalloffSize=0.45
    OptionalParentItem=(Type=Bone,Name="pelvis")
    RotationOffset=(X=180,Y=180,Z=0)
   )
   > &RigUnit_ModifyTransforms(
    ItemToModify=((Item=(Type=Bone,Name="dyn_skirt_fr_l_base"),Transform=(Rotation=(X=0,Y=0,Z=0,W=1),Translation=(X=0,Y=0,Z=0),Scale3D=(X=1,Y=1,Z=1))))
    ItemToModify.0.Transform.Rotation=&RigVMFunction_MathQuaternionFromEuler(Euler=(X=0,Y=0,Z=0)).Result
    Weight=@n7.OutputParam
   )
   > &RigUnit_SphericalPoseReader(
    ActiveRegionScaleFactors=(PositiveWidth=0,NegativeWidth=5.5,PositiveHeight=5,NegativeHeight=4.5)
    ActiveRegionSize=0.1
    Debug=(bDrawDebug=false,DebugScale=25,DebugSegments=20,DebugThickness=0.25)
    DriverAxis=(X=1,Y=0,Z=0)
    DriverItem=(Type=Bone,Name="thigh_r")
    FalloffRegionScaleFactors=(PositiveWidth=0,NegativeWidth=-1,PositiveHeight=0,NegativeHeight=-1)
    FalloffSize=0.45
    OptionalParentItem=(Type=Bone,Name="pelvis")
    RotationOffset=(X=0,Y=180,Z=0)
   )
  ]}
]
hierarchy [
 root {loc=<0 0 0>, rot=<0 0 0 1>}
 pelvis {parent=root, loc=<0 -0.5301 98.0571>, rot=<0.0127 0.707 -0.0127 -0.707>}
 spine_01 {parent=pelvis, loc=<6.1208 -0.9328 0>, rot=<0 0 0.034 -0.9994>}
 spine_02 {parent=spine_01, loc=<6.5996 -0.2095 0>, rot=<0 0 0.0238 -0.9997>}
 spine_03 {parent=spine_02, loc=<6.4914 -0.1237 0>, rot=<0 0 0.0149 -0.9999>}
 spine_04 {parent=spine_03, loc=<6.9598 0.0794 0>, rot=<0 0 -0.0132 -0.9999>}
 spine_05 {parent=spine_04, loc=<6.9987 0 0>, rot=<0 0 -0.0015 -1>}
 attach_cape {parent=spine_05, loc=<9.3216 -8.4501 0>, rot=<0.0616 -0.7044 -0.0616 -0.7044>}
 neck_01 {parent=spine_05, loc=<11.3138 0.1529 0>, rot=<0 0 -0.2003 -0.9797>}
 neck_02 {parent=neck_01, loc=<4.5627 0 0>, rot=<0 0 0.0504 -0.9987>}
 head {parent=neck_02, loc=<5.3914 0 0>, rot=<0 0 0.115 -0.9934>}
 attach_backpack {parent=spine_05, loc=<0 -8.35 0>, rot=<-0.0093 -0.707 0.0093 -0.707>}
 clavicle_l {parent=spine_05, loc=<5.4831 -0.1053 -3.2946>, rot=<0.0465 -0.7289 -0.009 -0.6829>}
 upperarm_l {parent=clavicle_l, loc=<9.8658 -1.2136 -0.4105>, rot=<0.0164 -0.3245 -0.0269 -0.9454>}
 lowerarm_l {parent=upperarm_l, loc=<26.465 -0.8345 0.259>, rot=<0.0036 0.0161 -0.2298 -0.9731>}
 hand_l {parent=lowerarm_l, loc=<22.7355 -0.3203 -0.6105>, rot=<-0.6784 -0.0223 0.0735 -0.7307>}
 thumb_01_l {parent=hand_l, loc=<1.3118 -0.3191 -2.459>, rot=<0.6546 -0.4042 0.0664 -0.6354>}
 thumb_02_l {parent=thumb_01_l, loc=<5.9336 -0.0448 0.0401>, rot=<-0.0037 0.0213 0.298 -0.9543>}
 thumb_03_l {parent=thumb_02_l, loc=<2.2223 0.3682 -0.0907>, rot=<0.0118 -0.013 0.0066 -0.9998>}
 index_metacarpal_l {parent=hand_l, loc=<1.6535 0.0183 -2.96>, rot=<0.0158 -0.0124 -0.0173 -0.9996>}
 index_01_l {parent=index_metacarpal_l, loc=<8.3008 0.0357 -0.0524>, rot=<-0.0044 0.0485 0.0924 -0.9945>}
 index_02_l {parent=index_01_l, loc=<3.3587 -0.3004 0.0452>, rot=<-0.002 0.0027 0.3225 -0.9466>}
 index_03_l {parent=index_02_l, loc=<2.5578 -0.123 -0.0252>, rot=<-0.002 0.0067 0.2558 -0.9667>}
 middle_metacarpal_l {parent=hand_l, loc=<2.1856 -0.1725 -0.5984>, rot=<0.006 0.0168 -0.0329 -0.9993>}
 middle_01_l {parent=middle_metacarpal_l, loc=<7.3275 -0.2447 0.0129>, rot=<-0.0208 0.0042 0.0858 -0.9961>}
 middle_02_l {parent=middle_01_l, loc=<4.0123 0.1401 0.1175>, rot=<-0.0218 0.0258 0.3512 -0.9357>}
 middle_03_l {parent=middle_02_l, loc=<2.6077 0.0534 0.1373>, rot=<0.0082 0.0188 0.1708 -0.9851>}
 ring_metacarpal_l {parent=hand_l, loc=<2.2074 -0.3625 1.3638>, rot=<-0.0284 0.0722 -0.0458 -0.9959>}
 ring_01_l {parent=ring_metacarpal_l, loc=<7.0295 -0.6197 -0.1446>, rot=<-0.0489 -0.0252 0.1892 -0.9804>}
 ring_02_l {parent=ring_01_l, loc=<3.6174 0.4647 -0.1027>, rot=<0 0 0.2806 -0.9598>}
 ring_03_l {parent=ring_02_l, loc=<2.4152 0.114 0.0199>, rot=<0 0 0.1189 -0.9929>}
 pinky_metacarpal_l {parent=hand_l, loc=<1.7616 -0.9753 3.3814>, rot=<-0.0397 0.0814 -0.069 -0.9935>}
 pinky_01_l {parent=pinky_metacarpal_l, loc=<7.1709 -0.4831 -0.1094>, rot=<-0.0654 -0.0035 0.2242 -0.9723>}
 pinky_02_l {parent=pinky_01_l, loc=<2.8274 0.0893 -0.1043>, rot=<0.0037 0.0011 0.2854 -0.9584>}
 pinky_03_l {parent=pinky_02_l, loc=<2.2046 0.0307 -0.0078>, rot=<-0.0002 0.0098 0.1445 -0.9895>}
 weapon_l {parent=hand_l, loc=<-0.1974 -3.4111 0.3811>, rot=<0 0 0 1>}
 hand_attach_l {parent=hand_l, loc=<8.5 -3 0>, rot=<0.7071 0.7071 0 0>}
 lowerarm_twist_02_l {parent=lowerarm_l, loc=<10.6707 0 0>, rot=<0 0 0 1>}
 lowerarm_twist_01_l {parent=lowerarm_l, loc=<18.316 0 0>, rot=<0 0 0 1>}
 upperarm_twist_01_l {parent=upperarm_l, loc=<7.2135 -0.1134 0>, rot=<0 0 0 1>}
 upperarm_twist_02_l {parent=upperarm_l, loc=<19.7998 -0.1134 0>, rot=<0 0 0 1>}
 clavicle_r {parent=spine_05, loc=<5.4831 -0.1053 3.2946>, rot=<0.7289 0.0465 -0.6829 0.009>}
 upperarm_r {parent=clavicle_r, loc=<-9.8658 1.2136 0.4105>, rot=<0.0164 -0.3245 -0.0269 -0.9454>}
 lowerarm_r {parent=upperarm_r, loc=<-26.465 0.8345 -0.259>, rot=<0.0036 0.0161 -0.2298 -0.9731>}
 hand_r {parent=lowerarm_r, loc=<-22.7355 0.3203 0.6106>, rot=<-0.6784 -0.0223 0.0735 -0.7307>}
 thumb_01_r {parent=hand_r, loc=<-1.3118 0.3191 2.459>, rot=<0.6546 -0.4042 0.0664 -0.6354>}
 thumb_02_r {parent=thumb_01_r, loc=<-5.9336 0.0448 -0.0401>, rot=<-0.0037 0.0213 0.298 -0.9543>}
 thumb_03_r {parent=thumb_02_r, loc=<-2.2223 -0.3682 0.0907>, rot=<0.0118 -0.013 0.0066 -0.9998>}
 index_metacarpal_r {parent=hand_r, loc=<-1.6535 -0.0183 2.96>, rot=<0.0158 -0.0124 -0.0173 -0.9996>}
 index_01_r {parent=index_metacarpal_r, loc=<-8.3008 -0.0357 0.0524>, rot=<-0.0044 0.0485 0.0924 -0.9945>}
 index_02_r {parent=index_01_r, loc=<-3.3587 0.3004 -0.0452>, rot=<-0.002 0.0027 0.3225 -0.9466>}
 index_03_r {parent=index_02_r, loc=<-2.5578 0.123 0.0252>, rot=<-0.002 0.0067 0.2558 -0.9667>}
 middle_metacarpal_r {parent=hand_r, loc=<-2.1856 0.1725 0.5984>, rot=<0.006 0.0168 -0.0329 -0.9993>}
 middle_01_r {parent=middle_metacarpal_r, loc=<-7.3275 0.2447 -0.0129>, rot=<-0.0208 0.0042 0.0858 -0.9961>}
 middle_02_r {parent=middle_01_r, loc=<-4.0123 -0.1401 -0.1175>, rot=<-0.0218 0.0258 0.3512 -0.9357>}
 middle_03_r {parent=middle_02_r, loc=<-2.6077 -0.0534 -0.1373>, rot=<0.0082 0.0188 0.1708 -0.9851>}
 ring_metacarpal_r {parent=hand_r, loc=<-2.2074 0.3625 -1.3638>, rot=<-0.0284 0.0722 -0.0458 -0.9959>}
 ring_01_r {parent=ring_metacarpal_r, loc=<-7.0295 0.6197 0.1446>, rot=<-0.0489 -0.0252 0.1892 -0.9804>}
 ring_02_r {parent=ring_01_r, loc=<-3.6174 -0.4647 0.1027>, rot=<0 0 0.2806 -0.9598>}
 ring_03_r {parent=ring_02_r, loc=<-2.4152 -0.114 -0.0199>, rot=<0 0 0.1189 -0.9929>}
 pinky_metacarpal_r {parent=hand_r, loc=<-1.7616 0.9753 -3.3814>, rot=<-0.0397 0.0814 -0.069 -0.9935>}
 pinky_01_r {parent=pinky_metacarpal_r, loc=<-7.1709 0.4831 0.1094>, rot=<-0.0654 -0.0035 0.2242 -0.9723>}
 pinky_02_r {parent=pinky_01_r, loc=<-2.8274 -0.0893 0.1043>, rot=<0.0037 0.0011 0.2854 -0.9584>}
 pinky_03_r {parent=pinky_02_r, loc=<-2.2046 -0.0307 0.0078>, rot=<-0.0002 0.0098 0.1445 -0.9895>}
 weapon_r {parent=hand_r, loc=<0.1974 3.4111 -0.3811>, rot=<0 0 0 1>}
 hand_attach_r {parent=hand_r, loc=<-8.5001 2.9999 0>, rot=<0 0 -0.7071 -0.7071>}
 lowerarm_twist_02_r {parent=lowerarm_r, loc=<-10.6707 0 0>, rot=<0 0 0 -1>}
 lowerarm_twist_01_r {parent=lowerarm_r, loc=<-18.316 0 0>, rot=<0 0 0 -1>}
 upperarm_twist_01_r {parent=upperarm_r, loc=<-7.2135 0.1134 0>, rot=<0 0 0 -1>}
 upperarm_twist_02_r {parent=upperarm_r, loc=<-19.7998 0.1134 0>, rot=<0 0 0 -1>}
 dyn_neckBow_base {parent=spine_05, loc=<11.1101 -4.8116 0>, rot=<0.5196 -0.4795 0.4795 -0.5196>}
 dyn_neckBow_center {parent=dyn_neckBow_base, loc=<0 0 0>, rot=<0.005 -0.475 0.0058 -0.88>}
 dyn_neckBow_loop_long_01 {parent=dyn_neckBow_center, loc=<4.344 1.6864 -0.8455>, rot=<-0.0444 -0.1896 -0.1371 -0.9712>}
 dyn_neckBow_loop_long_02 {parent=dyn_neckBow_loop_long_01, loc=<4.4169 0 0>, rot=<0 0 0 1>}
 dyn_neckBow_loop_short_01 {parent=dyn_neckBow_center, loc=<4.5374 -1.9583 -0.6535>, rot=<0.1828 -0.1527 0.1992 -0.9506>}
 dyn_neckBow_tail_long_01 {parent=dyn_neckBow_center, loc=<4.6072 0.8458 -1.1053>, rot=<0.0176 -0.2275 -0.0243 -0.9733>}
 dyn_neckBow_tail_long_02 {parent=dyn_neckBow_tail_long_01, loc=<6.9969 0 0>, rot=<-0.0236 -0.1104 0.001 -0.9936>}
 dyn_neckBow_tail_long_03 {parent=dyn_neckBow_tail_long_02, loc=<5.5545 0 0>, rot=<0 0 0 1>}
 dyn_neckBow_tail_short_01 {parent=dyn_neckBow_center, loc=<4.7042 -0.2668 -1.1952>, rot=<0.0321 -0.2067 0.0402 -0.9771>}
 dyn_neckBow_tail_short_02 {parent=dyn_neckBow_tail_short_01, loc=<4.5722 0 0>, rot=<0 0 0 1>}
 dyn_necklace_01 {parent=spine_05, loc=<-2.2668 12.1781 0>, rot=<-0.1215 -0.6966 -0.6966 -0.1215>}
 dyn_necklace_02 {parent=dyn_necklace_01, loc=<5.1051 0 0>, rot=<0 -0.1653 0 -0.9862>}
 dyn_necklace_03 {parent=dyn_necklace_02, loc=<4.7157 0 0>, rot=<0 0 0 1>}
 dyn_skirtBow_L_base {parent=spine_01, loc=<0.9215 6.1897 -9.6855>, rot=<-0.2311 -0.6683 -0.2525 -0.6605>}
 dyn_skirtBow_L_center {parent=dyn_skirtBow_L_base, loc=<0 0 0>, rot=<0.0073 -0.4697 -0.0301 -0.8823>}
 dyn_skirtBow_L_loop_r {parent=dyn_skirtBow_L_center, loc=<4.8847 1.1862 -0.5531>, rot=<0 0 -0.1305 -0.9914>}
 dyn_skirtBow_L_tail_r_01 {parent=dyn_skirtBow_L_center, loc=<4.1466 0.6945 -1.197>, rot=<0.0457 -0.0101 -0.0209 -0.9987>}
 dyn_skirtBow_L_tail_r_02 {parent=dyn_skirtBow_L_tail_r_01, loc=<3.9716 0 0>, rot=<0 0 0 1>}
 dyn_skirtBow_L_loopTail_l_01 {parent=dyn_skirtBow_L_center, loc=<4.2783 -0.6446 -0.9648>, rot=<0.0606 0.0092 -0.0324 -0.9976>}
 dyn_skirtBow_L_loopTail_l_02 {parent=dyn_skirtBow_L_loopTail_l_01, loc=<5.039 0 0>, rot=<0 0 0.0523 -0.9986>}
 dyn_skirtBow_R_base {parent=spine_01, loc=<0.9215 6.1897 9.685>, rot=<0.6605 0.2525 0.6683 0.2311>}
 dyn_skirtBow_R_center {parent=dyn_skirtBow_R_base, loc=<0 0 0>, rot=<-0.0145 -0.3428 -0.0125 -0.9392>}
 dyn_skirtBow_R_loop_r {parent=dyn_skirtBow_R_center, loc=<4.3272 1.3618 -1.1764>, rot=<0 0 -0.0436 -0.999>}
 dyn_skirtBow_R_loop_l {parent=dyn_skirtBow_R_center, loc=<4.4236 -1.5705 -1.8814>, rot=<0 0 0.1736 -0.9848>}
 dyn_skirtBow_R_tail_r_01 {parent=dyn_skirtBow_R_center, loc=<4.449 0.4656 -1.6351>, rot=<-0.0531 -0.1028 0.0758 -0.9904>}
 dyn_skirtBow_R_tail_r_02 {parent=dyn_skirtBow_R_tail_r_01, loc=<4.6393 0 0>, rot=<0 0 -0.0175 -0.9998>}
 dyn_skitBow_R_tail_l_01 {parent=dyn_skirtBow_R_center, loc=<3.3284 -0.8951 -2.2859>, rot=<-0.0813 -0.166 0.0918 -0.9785>}
 dyn_skitBow_R_tail_l_02 {parent=dyn_skitBow_R_tail_l_01, loc=<4.4101 0 0>, rot=<0 0 0.0262 -0.9997>}
 thigh_l {parent=pelvis, loc=<-1.0855 -0.2109 -8.7053>, rot=<0.0009 0.0355 0.0155 -0.9993>}
 calf_l {parent=thigh_l, loc=<-42.5284 -0.0019 -0.7297>, rot=<0.0006 -0.0161 -0.0391 -0.9991>}
 foot_l {parent=calf_l, loc=<-46.7401 -2.5014 -0.592>, rot=<0.0427 -0.0238 0.0353 -0.9982>}
 ball_l {parent=foot_l, loc=<-3.5658 10.8066 -0.1544>, rot=<0.0001 -0.0001 -0.7063 -0.7079>}
 calf_twist_02_l {parent=calf_l, loc=<-19.1546 -1.7943 -0.2638>, rot=<0.0237 0.0095 -0.0035 -0.9997>}
 calf_twist_01_l {parent=calf_l, loc=<-35.8483 -1.9197 -0.5794>, rot=<0.0237 0.0095 -0.0035 -0.9997>}
 thigh_twist_01_l {parent=thigh_l, loc=<-27.952 0 0>, rot=<0.0013 0 0 -1>}
 thigh_r {parent=pelvis, loc=<-1.0855 -0.2109 8.7053>, rot=<0.0355 -0.0009 0.9993 0.0155>}
 calf_r {parent=thigh_r, loc=<42.5284 0.0019 0.7297>, rot=<0.0006 -0.0161 -0.0391 -0.9991>}
 foot_r {parent=calf_r, loc=<46.7401 2.5014 0.592>, rot=<0.0427 -0.0238 0.0353 -0.9982>}
 ball_r {parent=foot_r, loc=<3.5658 -10.8066 0.1544>, rot=<0.0001 -0.0001 -0.7063 -0.7079>}
 calf_twist_02_r {parent=calf_r, loc=<19.1546 1.7943 0.2638>, rot=<0.0237 0.0095 -0.0035 -0.9997>}
 calf_twist_01_r {parent=calf_r, loc=<35.8483 1.9197 0.5794>, rot=<0.0237 0.0095 -0.0035 -0.9997>}
 thigh_twist_01_r {parent=thigh_r, loc=<27.952 0 0>, rot=<0.0013 0 0 -1>}
 skirt_dyn_joints {parent=pelvis, loc=<-97.9752 4.0424 0>, rot=<-0.0127 -0.707 0.0127 -0.707>}
 dyn_skirt_fr_r_base {parent=skirt_dyn_joints, loc=<-8.1397 7.4588 94.7815>, rot=<0 0 0.887 0.4617>}
 dyn_skirt_fr_r_01 {parent=dyn_skirt_fr_r_base, loc=<0 0 0>, rot=<-0.0276 -0.5902 -0.0074 -0.8067>}
 dyn_skirt_fr_r_02 {parent=dyn_skirt_fr_r_01, loc=<3.8946 0 0>, rot=<0.0275 0.1651 0.0158 -0.9858>}
 dyn_skirt_fr_r_top {parent=dyn_skirt_fr_r_02, loc=<3.7497 0 0>, rot=<0 0 0 1>}
 dyn_skirt_fr_r_03 {parent=dyn_skirt_fr_r_02, loc=<9.4179 0.3503 -3.2979>, rot=<0 -0.1305 0 -0.9914>}
 dyn_skirt_fr_l_base {parent=skirt_dyn_joints, loc=<8.7717 8.4145 94.4008>, rot=<0.0553 0.0258 -0.4218 -0.9046>}
 dyn_skirt_fr_l_01 {parent=dyn_skirt_fr_l_base, loc=<0 0 0>, rot=<0.0012 -0.5938 -0.0051 -0.8046>}
 dyn_skirt_fr_l_02 {parent=dyn_skirt_fr_l_01, loc=<4.4582 0 0>, rot=<-0.0168 -0.0308 -0.0027 -0.9994>}
 dyn_skirt_fr_l_top {parent=dyn_skirt_fr_l_02, loc=<2.4682 0.5681 2.071>, rot=<0.0298 0.3407 -0.0819 -0.9361>}
 dyn_skirt_fr_l_03 {parent=dyn_skirt_fr_l_02, loc=<8.4961 0.0452 0.1295>, rot=<0.0558 0.0049 0.0203 -0.9982>}
 dyn_skirt_side_r_base {parent=skirt_dyn_joints, loc=<-14.6684 0.1125 96.4184>, rot=<0 0 1 0>}
 dyn_skirt_side_r_01 {parent=dyn_skirt_side_r_base, loc=<0 0 0>, rot=<0 -0.5548 0 -0.832>}
 dyn_skirt_side_r_02 {parent=dyn_skirt_side_r_01, loc=<4.2479 0 0>, rot=<0.0253 0.1425 -0.0151 -0.9894>}
 dyn_skirt_side_r_top {parent=dyn_skirt_side_r_02, loc=<2.8135 0 0>, rot=<0 0 0 1>}
 dyn_skirt_side_r_03 {parent=dyn_skirt_side_r_02, loc=<8.7582 0.0093 -3.2353>, rot=<0 -0.1305 0 -0.9914>}
 dyn_skirt_side_l_base {parent=skirt_dyn_joints, loc=<15.452 0.1125 95.6741>, rot=<0 0 0 1>}
 dyn_skirt_side_l_01 {parent=dyn_skirt_side_l_base, loc=<0 0 0>, rot=<-0.0309 -0.5458 0.0473 -0.836>}
 dyn_skirt_side_l_02 {parent=dyn_skirt_side_l_01, loc=<4.1712 0 0>, rot=<0.0695 -0.0342 0.0166 -0.9969>}
 dyn_skirt_side_l_top {parent=dyn_skirt_side_l_02, loc=<2.9004 0.2537 1.6809>, rot=<-0.0113 0.2586 -0.0421 -0.965>}
 dyn_skirt_side_l_03 {parent=dyn_skirt_side_l_02, loc=<7.8371 0.4404 -0.2499>, rot=<-0.0038 0.0871 -0.0435 -0.9952>}
 dyn_skirt_bck_r_base {parent=skirt_dyn_joints, loc=<-11.7884 -9.5699 96.3715>, rot=<0 0 0.9397 -0.342>}
 dyn_skirt_bck_r_01 {parent=dyn_skirt_bck_r_base, loc=<0 0 0>, rot=<-0.0263 -0.5604 0.0388 -0.8269>}
 dyn_skirt_bck_r_02 {parent=dyn_skirt_bck_r_01, loc=<3.9702 0 0>, rot=<0.1177 0.1248 -0.0652 -0.983>}
 dyn_skirt_bck_r_top {parent=dyn_skirt_bck_r_02, loc=<3.1943 0 0>, rot=<0 0 0 1>}
 dyn_skirt_bck_r_03 {parent=dyn_skirt_bck_r_02, loc=<8.4691 -0.1932 -3.6861>, rot=<0 -0.1305 0 -0.9914>}
 dyn_skirt_bck_c_base {parent=skirt_dyn_joints, loc=<0 -12.8035 95.942>, rot=<0 0 0.7071 -0.7071>}
 dyn_skirt_bck_c_01 {parent=dyn_skirt_bck_c_base, loc=<0 0 0>, rot=<0 -0.5609 0 -0.8279>}
 dyn_skirt_bck_c_02 {parent=dyn_skirt_bck_c_01, loc=<4.2167 0 0>, rot=<0 -0.0752 0 -0.9972>}
 dyn_skirt_bck_c_top {parent=dyn_skirt_bck_c_02, loc=<3.0819 0 0.9068>, rot=<0 0.1305 0 -0.9914>}
 dyn_skirt_bck_c_03 {parent=dyn_skirt_bck_c_02, loc=<9.1739 0 0>, rot=<0 0 0 1>}
 dyn_skirt_bck_l_base {parent=skirt_dyn_joints, loc=<11.4351 -9.5335 96.0516>, rot=<0 0 0.4226 -0.9063>}
 dyn_skirt_bck_l_01 {parent=dyn_skirt_bck_l_base, loc=<0 0 0>, rot=<-0.0017 -0.5488 0.0026 -0.836>}
 dyn_skirt_bck_l_02 {parent=dyn_skirt_bck_l_01, loc=<4.1313 0 0>, rot=<0.0693 0.1372 -0.0422 -0.9872>}
 dyn_skirt_bck_l_top {parent=dyn_skirt_bck_l_02, loc=<3.525 0 0>, rot=<0 0 0 1>}
 dyn_skirt_bck_l_03 {parent=dyn_skirt_bck_l_02, loc=<8.2559 -0.3211 -3.2367>, rot=<0 -0.1305 0 -0.9914>}
 attach {parent=root, loc=<0 0 0>, rot=<0 0 0 1>}
 ik_hand_root {parent=root, loc=<0 0 0>, rot=<0 0 0 1>}
 ik_hand_gun {parent=ik_hand_root, loc=<-48.1743 3.3545 104.4381>, rot=<0.6276 0.1309 0.4183 -0.6434>}
 ik_hand_r {parent=ik_hand_gun, loc=<0 0 0>, rot=<0 0 0 1>}
 ik_hand_l {parent=ik_hand_gun, loc=<59.3271 67.6997 34.3524>, rot=<-0.6158 -0.7027 -0.3565 0>}
 ik_foot_root {parent=root, loc=<0 0 0>, rot=<0 0 0 1>}
 ik_foot_l {parent=ik_foot_root, loc=<14.8262 -7.3545 8.296>, rot=<0.0357 0.7045 0.0284 -0.7082>}
 ik_foot_r {parent=ik_foot_root, loc=<-14.8262 -7.3545 8.296>, rot=<0.7082 0.0284 -0.7045 0.0357>}
 attach_fp {parent=root, loc=<0 0 0>, rot=<0 0 0 1>}
 camera_root {parent=root, loc=<0 0 0>, rot=<0 0 0 1>}
 camera {parent=camera_root, loc=<0 0 0>, rot=<0 0 0 1>}
 warp_point_01 {parent=root, loc=<0 0 0>, rot=<0 0 0 1>}
 warp_point_02 {parent=root, loc=<0 0 0>, rot=<0 0 0 1>}
 warp_point_03 {parent=root, loc=<0 0 0>, rot=<0 0 0 1>}
]
