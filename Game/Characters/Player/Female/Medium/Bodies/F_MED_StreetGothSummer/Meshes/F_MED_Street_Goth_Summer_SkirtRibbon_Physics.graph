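# Physics asset for the two skirt bows on the female medium Street Goth Summer player body, using capsule bodies on the bow, front skirt, pelvis and spine bones.
# Each bow chains from spine_01 through base and center to branching loops and tails, with asymmetric swing limits and zero twist.
# Larger front skirt capsules chain rigidly to the pelvis with locked angular motion, while the spine-to-pelvis joint rotates freely.
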PhysicsAsset /Game/Characters/Player/Female/Medium/Bodies/F_MED_StreetGothSummer/Meshes/F_MED_Street_Goth_Summer_SkirtRibbon_Physics
bodies {
 dyn_skirtBow_L_base [{shapeType=Capsule, length=1, radius=1}]
 dyn_skirtBow_L_center [{shapeType=Capsule, length=1, radius=1}]
 dyn_skirtBow_L_loopTail_l_01 [{shapeType=Capsule, length=3.25, radius=1, center=<2.2498 0 0.1818>, rotation=<rot 80.0001 0.0004 -0.0002>}]
 dyn_skirtBow_L_loopTail_l_02 [{shapeType=Capsule, length=3.25, radius=1, center=<3.1114 0 0.0298>, rotation=<rot 80.0001 0.0004 -0.0002>}]
 dyn_skirtBow_L_loop_r [{shapeType=Capsule, length=3.25, radius=1, center=<1.9007 0 0.2433>, rotation=<rot 80.0001 0.0004 -0.0001>}]
 dyn_skirtBow_L_tail_r_01 [{shapeType=Capsule, length=3.25, radius=1, center=<1.9007 0 0.2433>, rotation=<rot 80.0001 0.0004 -0.0002>}]
 dyn_skirtBow_L_tail_r_02 [{shapeType=Capsule, length=3.25, radius=1, center=<1.9007 0 0.2433>, rotation=<rot 80.0001 0.0004 -0.0002>}]
 dyn_skirtBow_R_base [{shapeType=Capsule, length=1, radius=1}]
 dyn_skirtBow_R_center [{shapeType=Capsule, length=1, radius=1}]
 dyn_skirtBow_R_loop_l [{shapeType=Capsule, length=3.25, radius=1, center=<1.9007 0 0.2433>, rotation=<rot 80.0001 -0.0003 0.0002>}]
 dyn_skirtBow_R_loop_r [{shapeType=Capsule, length=3.25, radius=1, center=<1.9007 0 0.2433>, rotation=<rot 80.0001 -0.0003 0.0004>}]
 dyn_skirtBow_R_tail_r_01 [{shapeType=Capsule, length=3.25, radius=1, center=<1.9007 0 0.2433>, rotation=<rot 80.0001 -0.0004 0.0002>}]
 dyn_skirtBow_R_tail_r_02 [{shapeType=Capsule, length=3.25, radius=1, center=<2.3444 0 -0.2342>, rotation=<rot 70.0001 -0.0004 0.0002>}]
 dyn_skirt_fr_l_01 [{shapeType=Capsule, length=14.5, radius=4.25, center=<-4.6719 -2.8329 -3.2424>, rotation=<rot 0.0003 0.0002 -70.0001>}]
 dyn_skirt_fr_l_02 [{shapeType=Capsule, length=14.5, radius=4.25, center=<-3.2084 -2.8011 -2.0958>, rotation=<rot 0.0003 0.0002 -70.0001>}]
 dyn_skirt_fr_r_01 [{shapeType=Capsule, length=15.5, radius=4.25, center=<-4.1154 3.6279 -2.3571>, rotation=<rot 0.0004 -0.0003 70.0001>}]
 dyn_skirt_fr_r_02 [{shapeType=Capsule, length=15.25, radius=4.5, center=<-3.2084 3.8266 -0.1631>, rotation=<rot 0.0004 -0.0002 70.0002>}]
 dyn_skitBow_R_tail_l_01 [{shapeType=Capsule, length=3.25, radius=1, center=<1.9007 0 0.2433>, rotation=<rot 80.0001 -0.0004 0.0002>}]
 dyn_skitBow_R_tail_l_02 [{shapeType=Capsule, length=3.25, radius=1, center=<1.9007 0 0.2433>, rotation=<rot 80.0001 -0.0004 0.0002>}]
 pelvis [{shapeType=Capsule, length=12.75, radius=8.75, center=<1.4534 -1.784 0>}]
 spine_01 [{shapeType=Capsule, length=5.5, radius=8.5, center=<3.7623 1.3683 0>}]
}
constraints [
 {bone1Name=dyn_skirtBow_L_loopTail_l_01, bone2Name=dyn_skirtBow_L_center, swing1LimitDegrees=10, swing2LimitDegrees=45, twistLimitDegrees=0}
 {
  bone1Name=dyn_skirtBow_L_loopTail_l_02
  bone2Name=dyn_skirtBow_L_loopTail_l_01
  swing1LimitDegrees=10
  swing2LimitDegrees=45
  twistLimitDegrees=0
 }
 {bone1Name=dyn_skirtBow_L_loop_r, bone2Name=dyn_skirtBow_L_center, swing1LimitDegrees=10, swing2LimitDegrees=45, twistLimitDegrees=0}
 {bone1Name=dyn_skirtBow_L_tail_r_01, bone2Name=dyn_skirtBow_L_center, swing1LimitDegrees=10, swing2LimitDegrees=45, twistLimitDegrees=0}
 {bone1Name=dyn_skirtBow_L_tail_r_02, bone2Name=dyn_skirtBow_L_tail_r_01, swing1LimitDegrees=10, swing2LimitDegrees=45, twistLimitDegrees=0}
 {bone1Name=dyn_skirtBow_R_loop_l, bone2Name=dyn_skirtBow_R_center, swing1LimitDegrees=10, swing2LimitDegrees=45, twistLimitDegrees=0}
 {bone1Name=dyn_skirtBow_R_loop_r, bone2Name=dyn_skirtBow_R_center, swing1LimitDegrees=10, swing2LimitDegrees=45, twistLimitDegrees=0}
 {bone1Name=dyn_skirtBow_R_tail_r_01, bone2Name=dyn_skirtBow_R_center, swing1LimitDegrees=10, swing2LimitDegrees=45, twistLimitDegrees=0}
 {bone1Name=dyn_skirtBow_R_tail_r_02, bone2Name=dyn_skirtBow_R_tail_r_01, swing1LimitDegrees=10, swing2LimitDegrees=45, twistLimitDegrees=0}
 {bone1Name=dyn_skitBow_R_tail_l_01, bone2Name=dyn_skirtBow_R_center, swing1LimitDegrees=10, swing2LimitDegrees=45, twistLimitDegrees=0}
 {bone1Name=dyn_skitBow_R_tail_l_02, bone2Name=dyn_skitBow_R_tail_l_01, swing1LimitDegrees=10, swing2LimitDegrees=45, twistLimitDegrees=0}
 {bone1Name=dyn_skirtBow_R_center, bone2Name=dyn_skirtBow_R_base, swing1LimitDegrees=10, swing2LimitDegrees=45, twistLimitDegrees=0}
 {bone1Name=dyn_skirtBow_L_center, bone2Name=dyn_skirtBow_L_base, swing1LimitDegrees=10, swing2LimitDegrees=45, twistLimitDegrees=0}
 {bone1Name=dyn_skirtBow_L_base, bone2Name=spine_01, swing1LimitDegrees=10, swing2LimitDegrees=45, twistLimitDegrees=0}
 {bone1Name=dyn_skirtBow_R_base, bone2Name=spine_01, swing1LimitDegrees=10, swing2LimitDegrees=45, twistLimitDegrees=0}
 {
  bone1Name=spine_01
  bone2Name=pelvis
  swing1LimitDegrees=45
  swing1Motion=Free
  swing2LimitDegrees=45
  swing2Motion=Free
  twistLimitDegrees=45
  twistMotion=Free
 }
 {
  bone1Name=dyn_skirt_fr_l_01
  bone2Name=pelvis
  swing1LimitDegrees=45
  swing1Motion=Locked
  swing2LimitDegrees=45
  swing2Motion=Locked
  twistLimitDegrees=45
  twistMotion=Locked
 }
 {
  bone1Name=dyn_skirt_fr_r_01
  bone2Name=pelvis
  swing1LimitDegrees=45
  swing1Motion=Locked
  swing2LimitDegrees=45
  swing2Motion=Locked
  twistLimitDegrees=45
  twistMotion=Locked
 }
 {
  bone1Name=dyn_skirt_fr_r_02
  bone2Name=dyn_skirt_fr_r_01
  swing1LimitDegrees=45
  swing1Motion=Locked
  swing2LimitDegrees=45
  swing2Motion=Locked
  twistLimitDegrees=45
  twistMotion=Locked
 }
 {
  bone1Name=dyn_skirt_fr_l_02
  bone2Name=dyn_skirt_fr_l_01
  swing1LimitDegrees=45
  swing1Motion=Locked
  swing2LimitDegrees=45
  swing2Motion=Locked
  twistLimitDegrees=45
  twistMotion=Locked
 }
]
