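# Physics asset for the Street Goth Summer skirt on the medium female player body, defining capsule bodies and joints for skirt dynamics.
# Seven back, side and front chains hang from the pelvis (base, 01, 02), each forking at 02 into 03 and top bodies.
# Skirt joints have zero twist limits and varied swing limits, generally tighter along one axis lower down.
# Pelvis, spine_01 and thigh capsules represent the body, joined by free angular constraints.
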PhysicsAsset /Game/Characters/Player/Female/Medium/Bodies/F_MED_StreetGothSummer/Meshes/F_MED_Street_Goth_Summer_Skirt_Physics
bodies {
 dyn_skirt_bck_c_01 [{shapeType=Capsule, length=1, radius=1}]
 dyn_skirt_bck_c_02 [{shapeType=Capsule, length=2.75, radius=4, center=<2.8172 0 0>, rotation=<rot -0.0004 0 90>}]
 dyn_skirt_bck_c_03 [{shapeType=Capsule, length=7.75, radius=4.75, center=<3.5318 0 0>, rotation=<rot -0.0003 0 -90>}]
 dyn_skirt_bck_c_base [{shapeType=Capsule, length=1, radius=1}]
 dyn_skirt_bck_c_top [{shapeType=Capsule, length=3.5, radius=2.25, center=<1.3187 0 0>, rotation=<rot -0.0002 0 -90>}]
 dyn_skirt_bck_l_01 [{shapeType=Capsule, length=1, radius=1}]
 dyn_skirt_bck_l_02 [{shapeType=Capsule, length=2.75, radius=4, center=<2.8172 0 1.0212>, rotation=<rot -0.0002 0.0002 89.9998>}]
 dyn_skirt_bck_l_03 [{shapeType=Capsule, length=7.75, radius=4.75, center=<3.5318 0 0>, rotation=<rot -0.0002 0.0002 -90.0001>}]
 dyn_skirt_bck_l_base [{shapeType=Capsule, length=1, radius=1}]
 dyn_skirt_bck_l_top [{shapeType=Capsule, length=3.5, radius=2.25, center=<1.3187 0 0>, rotation=<rot -0.0001 0.0001 -90.0001>}]
 dyn_skirt_bck_r_01 [{shapeType=Capsule, length=1, radius=1}]
 dyn_skirt_bck_r_02 [{shapeType=Capsule, length=2.75, radius=4, center=<2.8172 0 1.0212>, rotation=<rot -0.0003 -0.0002 90.0001>}]
 dyn_skirt_bck_r_03 [{shapeType=Capsule, length=7.75, radius=4.75, center=<3.5318 0 0>, rotation=<rot -0.0003 -0.0002 -89.9999>}]
 dyn_skirt_bck_r_base [{shapeType=Capsule, length=1, radius=1}]
 dyn_skirt_bck_r_top [{shapeType=Capsule, length=3.5, radius=2.25, center=<1.3187 0 0>, rotation=<rot -0.0002 -0.0001 -89.9999>}]
 dyn_skirt_fr_l_01 [{shapeType=Capsule, length=1, radius=1}]
 dyn_skirt_fr_l_02 [{shapeType=Capsule, length=1, radius=1}]
 dyn_skirt_fr_l_03 [{shapeType=Capsule, length=7.75, radius=4.75, center=<3.5318 0 0>, rotation=<rot 0.0003 0.0002 -90.0001>}]
 dyn_skirt_fr_l_base [{shapeType=Capsule, length=1, radius=1}]
 dyn_skirt_fr_l_top [{shapeType=Capsule, length=3.5, radius=2.25, center=<1.1901 0 0>, rotation=<rot 0.0002 0.0001 -90.0001>}]
 dyn_skirt_fr_r_01 [{shapeType=Capsule, length=1, radius=1}]
 dyn_skirt_fr_r_02 [{shapeType=Capsule, length=1, radius=1}]
 dyn_skirt_fr_r_03 [{shapeType=Capsule, length=7.75, radius=4.75, center=<3.5318 0 0>, rotation=<rot 0.0003 -0.0002 -89.9999>}]
 dyn_skirt_fr_r_base [{shapeType=Capsule, length=1, radius=1}]
 dyn_skirt_fr_r_top [{shapeType=Capsule, length=3.5, radius=2.25, center=<1.1901 0 0>, rotation=<rot 0.0002 -0.0001 -89.9999>}]
 dyn_skirt_side_l_01 [{shapeType=Capsule, length=1, radius=1}]
 dyn_skirt_side_l_02 [{shapeType=Capsule, length=1, radius=1}]
 dyn_skirt_side_l_03 [{shapeType=Capsule, length=7.75, radius=4.75, center=<3.5318 0 0>, rotation=<rot 0 0.0003 -90.0002>}]
 dyn_skirt_side_l_base [{shapeType=Capsule, length=1, radius=1}]
 dyn_skirt_side_l_top [{shapeType=Capsule, length=3.5, radius=2.25, center=<1.1901 0 0>, rotation=<rot 0 0.0001 -90.0001>}]
 dyn_skirt_side_r_01 [{shapeType=Capsule, length=1, radius=1}]
 dyn_skirt_side_r_02 [{shapeType=Capsule, length=1, radius=1}]
 dyn_skirt_side_r_03 [{shapeType=Capsule, length=7.75, radius=4.75, center=<3.5318 0 0>, rotation=<rot 0 -0.0003 -89.9999>}]
 dyn_skirt_side_r_base [{shapeType=Capsule, length=1, radius=1}]
 dyn_skirt_side_r_top [{shapeType=Capsule, length=3.5, radius=2.25, center=<1.1901 0 0>, rotation=<rot 0 -0.0002 -89.9999>}]
 pelvis [{shapeType=Capsule, length=11.5, radius=6, shapeName=pelvis_capsule}]
 spine_01 [{shapeType=Capsule, length=3, radius=3, shapeName=spine_01_capsule}]
 thigh_l [{shapeType=Capsule, length=30, radius=6.5, shapeName=thigh_l_capsule, center=<-16.5115 -0.6135 0.8813>, rotation=<rot -90 180 179.9999>}]
 thigh_r [{
   shapeType=Capsule
   length=29.75
   radius=6.75
   shapeName=thigh_r_capsule
   center=<16.5115 1.2511 -2.0269>
   rotation=<rot 90 179.9999 -179.9999>
  }]
}
constraints [
 {
  bone1Name=thigh_l
  bone2Name=pelvis
  swing1LimitDegrees=45
  swing1Motion=Free
  swing2LimitDegrees=45
  swing2Motion=Free
  twistLimitDegrees=45
  twistMotion=Free
 }
 {
  bone1Name=thigh_r
  bone2Name=pelvis
  swing1LimitDegrees=45
  swing1Motion=Free
  swing2LimitDegrees=45
  swing2Motion=Free
  twistLimitDegrees=45
  twistMotion=Free
 }
 {
  bone1Name=spine_01
  bone2Name=pelvis
  swing1LimitDegrees=45
  swing1Motion=Free
  swing2LimitDegrees=45
  swing2Motion=Free
  twistLimitDegrees=45
  twistMotion=Free
 }
 {bone1Name=dyn_skirt_bck_c_base, bone2Name=pelvis, swing1LimitDegrees=45, swing2LimitDegrees=45, twistLimitDegrees=0}
 {bone1Name=dyn_skirt_bck_c_01, bone2Name=dyn_skirt_bck_c_base, swing1LimitDegrees=5, swing2LimitDegrees=30, twistLimitDegrees=0}
 {bone1Name=dyn_skirt_bck_c_02, bone2Name=dyn_skirt_bck_c_01, swing1LimitDegrees=20, swing2LimitDegrees=30, twistLimitDegrees=0}
 {bone1Name=dyn_skirt_bck_c_03, bone2Name=dyn_skirt_bck_c_02, swing1LimitDegrees=15, swing2LimitDegrees=35, twistLimitDegrees=0}
 {bone1Name=dyn_skirt_bck_c_top, bone2Name=dyn_skirt_bck_c_02, swing1LimitDegrees=15, swing2LimitDegrees=25, twistLimitDegrees=0}
 {bone1Name=dyn_skirt_bck_l_base, bone2Name=pelvis, swing1LimitDegrees=45, swing2LimitDegrees=45, twistLimitDegrees=0}
 {bone1Name=dyn_skirt_bck_l_01, bone2Name=dyn_skirt_bck_l_base, swing1LimitDegrees=45, swing2LimitDegrees=45, twistLimitDegrees=0}
 {bone1Name=dyn_skirt_bck_l_02, bone2Name=dyn_skirt_bck_l_01, swing1LimitDegrees=5, swing2LimitDegrees=30, twistLimitDegrees=0}
 {bone1Name=dyn_skirt_bck_l_03, bone2Name=dyn_skirt_bck_l_02, swing1LimitDegrees=10, swing2LimitDegrees=30, twistLimitDegrees=0}
 {bone1Name=dyn_skirt_bck_l_top, bone2Name=dyn_skirt_bck_l_02, swing1LimitDegrees=10, swing2LimitDegrees=25, twistLimitDegrees=0}
 {bone1Name=dyn_skirt_bck_r_base, bone2Name=pelvis, swing1LimitDegrees=45, swing2LimitDegrees=45, twistLimitDegrees=0}
 {bone1Name=dyn_skirt_bck_r_01, bone2Name=dyn_skirt_bck_r_base, swing1LimitDegrees=45, swing2LimitDegrees=45, twistLimitDegrees=0}
 {bone1Name=dyn_skirt_bck_r_02, bone2Name=dyn_skirt_bck_r_01, swing1LimitDegrees=5, swing2LimitDegrees=30, twistLimitDegrees=0}
 {bone1Name=dyn_skirt_bck_r_03, bone2Name=dyn_skirt_bck_r_02, swing1LimitDegrees=10, swing2LimitDegrees=30, twistLimitDegrees=0}
 {bone1Name=dyn_skirt_bck_r_top, bone2Name=dyn_skirt_bck_r_02, swing1LimitDegrees=10, swing2LimitDegrees=25, twistLimitDegrees=0}
 {bone1Name=dyn_skirt_fr_l_base, bone2Name=pelvis, swing1LimitDegrees=45, swing2LimitDegrees=45, twistLimitDegrees=0}
 {bone1Name=dyn_skirt_fr_l_01, bone2Name=dyn_skirt_fr_l_base, swing1LimitDegrees=45, swing2LimitDegrees=45, twistLimitDegrees=0}
 {bone1Name=dyn_skirt_fr_l_02, bone2Name=dyn_skirt_fr_l_01, swing1LimitDegrees=45, swing2LimitDegrees=45, twistLimitDegrees=0}
 {bone1Name=dyn_skirt_fr_l_03, bone2Name=dyn_skirt_fr_l_02, swing1LimitDegrees=5, swing2LimitDegrees=30, twistLimitDegrees=0}
 {bone1Name=dyn_skirt_fr_l_top, bone2Name=dyn_skirt_fr_l_02, swing1LimitDegrees=5, swing2LimitDegrees=25, twistLimitDegrees=0}
 {bone1Name=dyn_skirt_fr_r_base, bone2Name=pelvis, swing1LimitDegrees=45, swing2LimitDegrees=45, twistLimitDegrees=0}
 {bone1Name=dyn_skirt_fr_r_01, bone2Name=dyn_skirt_fr_r_base, swing1LimitDegrees=45, swing2LimitDegrees=45, twistLimitDegrees=0}
 {bone1Name=dyn_skirt_fr_r_02, bone2Name=dyn_skirt_fr_r_01, swing1LimitDegrees=45, swing2LimitDegrees=45, twistLimitDegrees=0}
 {bone1Name=dyn_skirt_fr_r_03, bone2Name=dyn_skirt_fr_r_02, swing1LimitDegrees=5, swing2LimitDegrees=30, twistLimitDegrees=0}
 {bone1Name=dyn_skirt_fr_r_top, bone2Name=dyn_skirt_fr_r_02, swing1LimitDegrees=5, swing2LimitDegrees=25, twistLimitDegrees=0}
 {bone1Name=dyn_skirt_side_l_base, bone2Name=pelvis, swing1LimitDegrees=45, swing2LimitDegrees=45, twistLimitDegrees=0}
 {bone1Name=dyn_skirt_side_l_01, bone2Name=dyn_skirt_side_l_base, swing1LimitDegrees=45, swing2LimitDegrees=45, twistLimitDegrees=0}
 {bone1Name=dyn_skirt_side_l_02, bone2Name=dyn_skirt_side_l_01, swing1LimitDegrees=45, swing2LimitDegrees=45, twistLimitDegrees=0}
 {bone1Name=dyn_skirt_side_l_03, bone2Name=dyn_skirt_side_l_02, swing1LimitDegrees=5, swing2LimitDegrees=30, twistLimitDegrees=0}
 {bone1Name=dyn_skirt_side_l_top, bone2Name=dyn_skirt_side_l_02, swing1LimitDegrees=5, swing2LimitDegrees=25, twistLimitDegrees=0}
 {bone1Name=dyn_skirt_side_r_base, bone2Name=pelvis, swing1LimitDegrees=45, swing2LimitDegrees=45, twistLimitDegrees=0}
 {bone1Name=dyn_skirt_side_r_01, bone2Name=dyn_skirt_side_r_base, swing1LimitDegrees=45, swing2LimitDegrees=45, twistLimitDegrees=0}
 {bone1Name=dyn_skirt_side_r_02, bone2Name=dyn_skirt_side_r_01, swing1LimitDegrees=45, swing2LimitDegrees=45, twistLimitDegrees=0}
 {bone1Name=dyn_skirt_side_r_03, bone2Name=dyn_skirt_side_r_02, swing1LimitDegrees=5, swing2LimitDegrees=30, twistLimitDegrees=0}
 {bone1Name=dyn_skirt_side_r_top, bone2Name=dyn_skirt_side_r_02, swing1LimitDegrees=5, swing2LimitDegrees=25, twistLimitDegrees=0}
]
