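# Animation Blueprint for the female Street Goth Summer body, adding secondary motion to skirt, ribbon bows, necklace and neck bow.
# Parent pose feeds a Control Rig, then bone constraints, additive offsets, two pelvis-space RigidBody simulations and spring-based AnimDynamics chains.
# The event graph derives gravity and alpha values from movement states, thigh and head rotations, and backbling checks.
# It also drives material cloth-sim wind, with a bone-motion turbulence function.
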AnimBlueprint /Game/Characters/Player/Female/Medium/Bodies/F_MED_StreetGothSummer/Meshes/F_MED_Street_Goth_Summer_AnimBP
target_skeleton=&/F_MED_Street_Goth_Summer_Skeleton
graphs [
 AnimGraph {nodes [
   n1 = Control Rig(Source=Copy Pose From Mesh() {bUseAttachedParent=true}) {
    alphaBoolBlend {blendOption=Linear}
    controlRigAssetReference {blueprintRigClass=&/F_MED_Street_Goth_Summer_CtrlRig.F_MED_Street_Goth_Summer_CtrlRig_C}
   }
   n2 = "Constraint - dyn_skirt_fr_l_base (P)"(ComponentPose=Local To Component(LocalPose=@n1), Alpha=1, ConstraintWeights_0=1) {
    alphaBoolBlend {blendOption=Linear}
    boneToModify {boneName=dyn_skirt_fr_l_base}
    constraintSetup [{targetBone {boneName=thigh_l}}]
    constraintWeights [1]
   }
   n3 = "Constraint - dyn_skirt_fr_r_base (P)"(ComponentPose=@n2, Alpha=1, ConstraintWeights_0=1) {
    alphaBoolBlend {blendOption=Linear}
    boneToModify {boneName=dyn_skirt_fr_r_base}
    constraintSetup [{targetBone {boneName=thigh_r}}]
    constraintWeights [1]
   }
   n4 = RigidBody!(ComponentPose=@n3, Alpha, OverrideWorldGravity:Vector="0,0,-980", ExternalForce) -> (Pose) {
    bClampLinearTranslationLimitToRefPose=true
    bOverrideWorldGravity=true
    overridePhysicsAsset=&/F_MED_Street_Goth_Summer_Skirt_Physics
    simulationSpace=BaseBoneSpace
    alphaBoolBlend {blendOption=Linear}
    baseBoneRef {boneName=pelvis}
    componentAppliedLinearAccClamp=<3000 3000 3000>
    componentLinearAccScale=<0.7 0.7 0.7>
    componentLinearVelScale=<0.2 0.2 0.2>
    overrideWorldGravity {z=-980}
   }
   n5 = "Transform (Modify) Bone - Bone: dyn_skirt_fr_r_base"(ComponentPose=@n4, Alpha=1, Translation="-5.719,-1.464,0.336", Rotation="10.000,-10.000,0.000", Scale="1,1,1") {
    rotationMode=BMM_Additive
    scaleMode=BMM_Replace
    translationMode=BMM_Additive
    alphaBoolBlend {blendOption=Linear}
    alphaScaleBiasClamp {bInterpResult=true}
    boneToModify {boneName=dyn_skirt_fr_r_base}
    rotation=<rot 10 -10 0>
    translation=<-5.719 -1.464 0.336>
   }
   n6 = "Transform (Modify) Bone - Bone: dyn_skirt_side_r_base"(ComponentPose=@n5, Alpha=1, Translation="-4.495,0.000,0.000", Rotation="29.836,-8.008,10.843", Scale="1,1,1") {
    rotationMode=BMM_Additive
    scaleMode=BMM_Replace
    translationMode=BMM_Additive
    alphaBoolBlend {blendOption=Linear}
    alphaScaleBiasClamp {bInterpResult=true}
    boneToModify {boneName=dyn_skirt_side_r_base}
    rotation=<rot 29.836 -8.008 10.843>
    translation {x=-4.495}
   }
   n7 = "Transform (Modify) Bone - Bone: dyn_skirt_fr_l_base"(ComponentPose=@n6, Alpha=1, Translation="1.160,4.980,1.070", Rotation="-0.000,0.000,20.000", Scale="1,1,1") {
    rotationMode=BMM_Additive
    scaleMode=BMM_Replace
    translationMode=BMM_Additive
    alphaBoolBlend {blendOption=Linear}
    alphaScaleBiasClamp {bInterpResult=true}
    boneToModify {boneName=dyn_skirt_fr_l_base}
    rotation {roll=20}
    translation=<1.16 4.98 1.07>
   }
   n8 = "Transform (Modify) Bone - Bone: dyn_skirt_side_l_base"(ComponentPose=@n7, Alpha=1, Translation="-0.000,-0.000,1.482", Rotation="-27.187,-6.930,-1.884", Scale="1,1,1") {
    rotationMode=BMM_Additive
    scaleMode=BMM_Replace
    translationMode=BMM_Additive
    alphaBoolBlend {blendOption=Linear}
    alphaScaleBiasClamp {bInterpResult=true}
    boneToModify {boneName=dyn_skirt_side_l_base}
    rotation=<rot -27.187 -6.93 -1.884>
    translation {z=1.482}
   }
   n9 = "Transform (Modify) Bone - Bone: dyn_skirt_side_r_base"(ComponentPose=@n8, Alpha=1, Translation="-1.443,0.000,0.000", Rotation="0,0,0", Scale="1,1,1") {
    rotationMode=BMM_Additive
    scaleMode=BMM_Replace
    translationMode=BMM_Additive
    alphaBoolBlend {blendOption=Linear}
    alphaScaleBiasClamp {bInterpResult=true}
    boneToModify {boneName=dyn_skirt_side_r_base}
    translation {x=-1.443}
   }
   n10 = "Transform (Modify) Bone - Bone: dyn_skirt_side_l_base"(ComponentPose=@n9, Alpha=1, Translation="-1.319,-0.000,0.000", Rotation="-6.409,-7.692,60.431", Scale="1,1,1") {
    rotationMode=BMM_Additive
    scaleMode=BMM_Replace
    translationMode=BMM_Additive
    alphaBoolBlend {blendOption=Linear}
    alphaScaleBiasClamp {bInterpResult=true}
    boneToModify {boneName=dyn_skirt_side_l_base}
    rotation=<rot -6.409 -7.692 60.431>
    translation {x=-1.319}
   }
   n11 = "Transform (Modify) Bone - Bone: dyn_skirt_side_r_base"(ComponentPose=@n10, Alpha=1, Translation="0,0,0", Rotation="-20.000,0.000,-0.000", Scale="1,1,1") {
    rotationMode=BMM_Additive
    scaleMode=BMM_Replace
    translationMode=BMM_Additive
    alphaBoolBlend {blendOption=Linear}
    alphaScaleBiasClamp {bInterpResult=true}
    boneToModify {boneName=dyn_skirt_side_r_base}
    rotation {pitch=-20}
   }
   n12 = "Transform (Modify) Bone - Bone: dyn_skirt_side_l_base"(ComponentPose=@n11, Alpha=1, Translation="2.614,0.000,2.047", Rotation="10.000,0.000,0.000", Scale="1,1,1") {
    rotationMode=BMM_Additive
    scaleMode=BMM_Replace
    translationMode=BMM_Additive
    alphaBoolBlend {blendOption=Linear}
    alphaScaleBiasClamp {bInterpResult=true}
    boneToModify {boneName=dyn_skirt_side_l_base}
    rotation {pitch=10}
    translation {x=2.614, z=2.047}
   }
   n13 = "Transform (Modify) Bone - Bone: dyn_skirt_bck_r_base"(ComponentPose=@n12, Alpha=1, Translation="-0.000,-1.293,0.000", Rotation="0,0,0", Scale="1,1,1") {
    rotationMode=BMM_Additive
    scaleMode=BMM_Replace
    translationMode=BMM_Additive
    alphaBoolBlend {blendOption=Linear}
    alphaScaleBiasClamp {bInterpResult=true}
    boneToModify {boneName=dyn_skirt_bck_r_base}
    translation=<0 -1.2933>
   }
   n14 = "Transform (Modify) Bone - Bone: dyn_skirt_bck_l_base"(ComponentPose=@n13, Alpha=1, Translation="0.911,-1.327,0.000", Rotation="0,0,0", Scale="1,1,1") {
    rotationMode=BMM_Additive
    scaleMode=BMM_Replace
    translationMode=BMM_Additive
    alphaBoolBlend {blendOption=Linear}
    alphaScaleBiasClamp {bInterpResult=true}
    boneToModify {boneName=dyn_skirt_bck_l_base}
    translation=<0.911 -1.327>
   }
   n15 = "Transform (Modify) Bone - Bone: dyn_skirt_side_l_base"(ComponentPose=@n14, Alpha=1, Translation="0,-1,0", Rotation="0,0,0", Scale="1,1,1") {
    rotationMode=BMM_Additive
    scaleMode=BMM_Replace
    translationMode=BMM_Additive
    alphaBoolBlend {blendOption=Linear}
    alphaScaleBiasClamp {bInterpResult=true}
    boneToModify {boneName=dyn_skirt_side_l_base}
    translation {y=-1}
   }
   n16 = "Transform (Modify) Bone - Bone: dyn_skirt_bck_l_base"(ComponentPose=@n15, Alpha=1, Translation="1,-2,0", Rotation="10.000,-0.000,10.000", Scale="1,1,1") {
    rotationMode=BMM_Additive
    scaleMode=BMM_Replace
    translationMode=BMM_Additive
    alphaBoolBlend {blendOption=Linear}
    alphaScaleBiasClamp {bInterpResult=true}
    boneToModify {boneName=dyn_skirt_bck_l_base}
    rotation=<rot 10 0 10>
    translation=<1 -2>
   }
   n17 = "Transform (Modify) Bone - Bone: dyn_skirt_bck_c_base"(ComponentPose=@n16, Alpha=1, Translation="0,0,0", Rotation="-0.000,0.000,10.000", Scale="1,1,1") {
    rotationMode=BMM_Additive
    scaleMode=BMM_Replace
    translationMode=BMM_Additive
    alphaBoolBlend {blendOption=Linear}
    alphaScaleBiasClamp {bInterpResult=true}
    boneToModify {boneName=dyn_skirt_bck_c_base}
    rotation {roll=10}
   }
   n18 = "Transform (Modify) Bone - Bone: dyn_skirt_bck_c_base"(ComponentPose=@n17, Alpha=1, Translation="0.000,-0.355,0.000", Rotation="-0.000,0.000,20.000", Scale="1,1,1") {
    rotationMode=BMM_Additive
    scaleMode=BMM_Replace
    translationMode=BMM_Additive
    alphaBoolBlend {blendOption=Linear}
    alphaScaleBiasClamp {bInterpResult=true}
    boneToModify {boneName=dyn_skirt_bck_c_base}
    rotation {roll=20}
    translation {y=-0.355}
   }
   n19 = "Transform (Modify) Bone - Bone: dyn_skirt_side_l_base"(ComponentPose=@n18, Alpha=1, Translation="1.493,-9.301,0.777", Rotation="20.000,0.000,0.000", Scale="1,1,1") {
    rotationMode=BMM_Additive
    scaleMode=BMM_Replace
    translationMode=BMM_Additive
    alphaBoolBlend {blendOption=Linear}
    alphaScaleBiasClamp {bInterpResult=true}
    boneToModify {boneName=dyn_skirt_side_l_base}
    rotation {pitch=20}
    translation=<1.493 -9.301 0.777>
   }
   n20 = "Transform (Modify) Bone - Bone: dyn_skirt_bck_l_base"(ComponentPose=@n19, Alpha=1, Translation="-2.689,0.000,0.000", Rotation="19.842,1.837,10.475", Scale="1,1,1") {
    rotationMode=BMM_Additive
    scaleMode=BMM_Replace
    translationMode=BMM_Additive
    alphaBoolBlend {blendOption=Linear}
    alphaScaleBiasClamp {bInterpResult=true}
    boneToModify {boneName=dyn_skirt_bck_l_base}
    rotation=<rot 19.842 1.837 10.475>
    translation {x=-2.689}
   }
   n21 = "Transform (Modify) Bone - Bone: dyn_skirt_side_r_base"(ComponentPose=@n20, Alpha=1, Translation="-0.000,-0.000,1.327", Rotation="0,0,0", Scale="1,1,1") {
    rotationMode=BMM_Additive
    scaleMode=BMM_Replace
    translationMode=BMM_Additive
    alphaBoolBlend {blendOption=Linear}
    alphaScaleBiasClamp {bInterpResult=true}
    boneToModify {boneName=dyn_skirt_side_r_base}
    translation {z=1.327}
   }
   n22 = "Transform (Modify) Bone - Bone: dyn_skirt_fr_r_03"(ComponentPose=@n21, Alpha=1, Translation="-0.000,0.000,1.187", Rotation="-10.000,0.000,-0.000", Scale="1,1,1") {
    rotationMode=BMM_Additive
    scaleMode=BMM_Replace
    translationMode=BMM_Additive
    alphaBoolBlend {blendOption=Linear}
    alphaScaleBiasClamp {bInterpResult=true}
    boneToModify {boneName=dyn_skirt_fr_r_03}
    rotation {pitch=-10}
    translation {z=1.187}
   }
   n23 = "Transform (Modify) Bone - Bone: dyn_skirt_side_r_base"(ComponentPose=@n22, Alpha=1, Translation="-2.627,0.000,0.000", Rotation="0,0,0", Scale="1,1,1") {
    rotationMode=BMM_Additive
    scaleMode=BMM_Replace
    translationMode=BMM_Additive
    alphaBoolBlend {blendOption=Linear}
    alphaScaleBiasClamp {bInterpResult=true}
    boneToModify {boneName=dyn_skirt_side_r_base}
    translation {x=-2.627}
   }
   n24 = "Transform (Modify) Bone - Bone: dyn_skirt_bck_r_base"(ComponentPose=@n23, Alpha=1, Translation="-1.446,0.000,-0.000", Rotation="-10.000,20.000,0.000", Scale="1,1,1") {
    rotationMode=BMM_Additive
    scaleMode=BMM_Replace
    translationMode=BMM_Additive
    alphaBoolBlend {blendOption=Linear}
    alphaScaleBiasClamp {bInterpResult=true}
    boneToModify {boneName=dyn_skirt_bck_r_base}
    rotation {pitch=-10, yaw=20}
    translation {x=-1.4463, z=0}
   }
   n25 = "Transform (Modify) Bone - Bone: dyn_skirt_bck_l_base"(ComponentPose=@n24, Alpha=1, Translation="1.918,0.685,0.000", Rotation="-0.000,20.000,20.000", Scale="1,1,1") {
    rotationMode=BMM_Additive
    scaleMode=BMM_Replace
    translationMode=BMM_Additive
    alphaBoolBlend {blendOption=Linear}
    alphaScaleBiasClamp {bInterpResult=true}
    boneToModify {boneName=dyn_skirt_bck_l_base}
    rotation {roll=20, yaw=20}
    translation=<1.918 0.685>
   }
   n26 = "Transform (Modify) Bone - Bone: dyn_skirt_bck_l_02"(ComponentPose=@n25, Alpha=1, Translation="0,0,0", Rotation="0,-10,-20", Scale="1,1,1") {
    rotationMode=BMM_Additive
    scaleMode=BMM_Replace
    translationMode=BMM_Additive
    alphaBoolBlend {blendOption=Linear}
    alphaScaleBiasClamp {bInterpResult=true}
    boneToModify {boneName=dyn_skirt_bck_l_02}
    rotation {roll=-20, yaw=-10}
   }
   n27 = "Constraint - dyn_skirt_side_l_02 (P)"(ComponentPose=@n26, Alpha=1, ConstraintWeights_0=1) {
    alphaBoolBlend {blendOption=Linear}
    alphaScaleBiasClamp {bInterpResult=true, interpSpeedDecreasing=1}
    boneToModify {boneName=dyn_skirt_side_l_02}
    constraintSetup [{targetBone {boneName=thigh_l}}]
    constraintWeights [1]
   }
   n28 = "Constraint - dyn_skirt_side_r_02 (P)"(ComponentPose=@n27, Alpha=1, ConstraintWeights_0=1) {
    alphaBoolBlend {blendOption=Linear}
    alphaScaleBiasClamp {bInterpResult=true, interpSpeedDecreasing=1}
    boneToModify {boneName=dyn_skirt_side_r_02}
    constraintSetup [{targetBone {boneName=thigh_r}}]
    constraintWeights [1]
   }
   n29 = "Constraint - dyn_skirtBow_L_center (R)"(ComponentPose=@n28, Alpha=1, ConstraintWeights_0=1) {
    alphaBoolBlend {blendOption=Linear}
    boneToModify {boneName=dyn_skirtBow_L_center}
    constraintSetup [{transformType=Rotation, targetBone {boneName=dyn_skirt_fr_l_base}}]
    constraintWeights [1]
   }
   n30 = "Constraint - dyn_skirtBow_R_center (R)"(ComponentPose=@n29, Alpha=1, ConstraintWeights_0=1) {
    alphaBoolBlend {blendOption=Linear}
    boneToModify {boneName=dyn_skirtBow_R_center}
    constraintSetup [{transformType=Rotation, targetBone {boneName=dyn_skirt_fr_r_base}}]
    constraintWeights [1]
   }
   n31 = "Constraint - dyn_skirtBow_L_loop_r (P)"(ComponentPose=@n30, Alpha=1, ConstraintWeights_0=1) {
    alphaBoolBlend {blendOption=Linear}
    alphaScaleBiasClamp {bInterpResult=true}
    boneToModify {boneName=dyn_skirtBow_L_loop_r}
    constraintSetup [{targetBone {boneName=dyn_skirt_fr_l_top}}]
    constraintWeights [1]
   }
   n32 = "Constraint - dyn_skirtBow_L_tail_r_01 (P)"(ComponentPose=@n31, Alpha=1, ConstraintWeights_0=1) {
    alphaBoolBlend {blendOption=Linear}
    alphaScaleBiasClamp {bInterpResult=true}
    boneToModify {boneName=dyn_skirtBow_L_tail_r_01}
    constraintSetup [{targetBone {boneName=dyn_skirt_fr_l_top}}]
    constraintWeights [1]
   }
   n33 = "Constraint - dyn_skirtBow_L_loopTail_l_01 (P)"(ComponentPose=@n32, Alpha=1, ConstraintWeights_0=1) {
    alphaBoolBlend {blendOption=Linear}
    alphaScaleBiasClamp {bInterpResult=true}
    boneToModify {boneName=dyn_skirtBow_L_loopTail_l_01}
    constraintSetup [{targetBone {boneName=dyn_skirt_fr_l_top}}]
    constraintWeights [1]
   }
   n34 = "Constraint - dyn_skirtBow_R_loop_r (P)"(ComponentPose=@n33, Alpha=1, ConstraintWeights_0=1) {
    alphaBoolBlend {blendOption=Linear}
    alphaScaleBiasClamp {bInterpResult=true}
    boneToModify {boneName=dyn_skirtBow_R_loop_r}
    constraintSetup [{targetBone {boneName=dyn_skirt_fr_r_top}}]
    constraintWeights [1]
   }
   n35 = "Constraint - dyn_skirtBow_R_loop_l (P)"(ComponentPose=@n34, Alpha=1, ConstraintWeights_0=1) {
    alphaBoolBlend {blendOption=Linear}
    alphaScaleBiasClamp {bInterpResult=true}
    boneToModify {boneName=dyn_skirtBow_R_loop_l}
    constraintSetup [{targetBone {boneName=dyn_skirt_fr_r_top}}]
    constraintWeights [1]
   }
   n36 = "Constraint - dyn_skirtBow_R_tail_r_01 (P)"(ComponentPose=@n35, Alpha=1, ConstraintWeights_0=1) {
    alphaBoolBlend {blendOption=Linear}
    alphaScaleBiasClamp {bInterpResult=true}
    boneToModify {boneName=dyn_skirtBow_R_tail_r_01}
    constraintSetup [{targetBone {boneName=dyn_skirt_fr_r_top}}]
    constraintWeights [1]
   }
   n37 = "Constraint - dyn_skitBow_R_tail_l_01 (P)"(ComponentPose=@n36, Alpha=1, ConstraintWeights_0=1) {
    alphaBoolBlend {blendOption=Linear}
    alphaScaleBiasClamp {bInterpResult=true}
    boneToModify {boneName=dyn_skitBow_R_tail_l_01}
    constraintSetup [{targetBone {boneName=dyn_skirt_fr_r_top}}]
    constraintWeights [1]
   }
   n38 = "Transform (Modify) Bone - Bone: dyn_skirtBow_L_base"(ComponentPose=@n37, Alpha=1, Translation="0.000,0.000,6.043", Rotation="0.000,0.000,-30.000", Scale="1,1,1") {
    rotationMode=BMM_Additive
    scaleMode=BMM_Replace
    translationMode=BMM_Additive
    alphaBoolBlend {blendOption=Linear}
    alphaScaleBiasClamp {bInterpResult=true}
    boneToModify {boneName=dyn_skirtBow_L_base}
    rotation {roll=-30}
    translation {z=6.043}
   }
   n39 = "Transform (Modify) Bone - Bone: dyn_skirtBow_R_base"(ComponentPose=@n38, Alpha=1, Translation="0.000,-0.866,4.370", Rotation="0.000,50.000,-0.000", Scale="1,1,1") {
    rotationMode=BMM_Additive
    scaleMode=BMM_Replace
    translationMode=BMM_Additive
    alphaBoolBlend {blendOption=Linear}
    alphaScaleBiasClamp {bInterpResult=true}
    boneToModify {boneName=dyn_skirtBow_R_base}
    rotation {yaw=50}
    translation {y=-0.866, z=4.37}
   }
   n40 = "Transform (Modify) Bone - Bone: dyn_skirtBow_L_loop_r"(ComponentPose=@n39, Alpha=1, Translation="0,0,0", Rotation="0,0,0", Scale="1,1,1") {
    rotationMode=BMM_Additive
    scaleMode=BMM_Replace
    translationMode=BMM_Additive
    alphaBoolBlend {blendOption=Linear}
    alphaScaleBiasClamp {bInterpResult=true}
    boneToModify {boneName=dyn_skirtBow_L_loop_r}
   }
   n41 = "Transform (Modify) Bone - Bone: dyn_skirtBow_R_loop_r"(ComponentPose=@n40, Alpha=1, Translation="0,0,0", Rotation="0,0,0", Scale="1,1,1") {
    rotationMode=BMM_Additive
    scaleMode=BMM_Replace
    translationMode=BMM_Additive
    alphaBoolBlend {blendOption=Linear}
    alphaScaleBiasClamp {bInterpResult=true}
    boneToModify {boneName=dyn_skirtBow_R_loop_r}
   }
   n42 = "Transform (Modify) Bone - Bone: dyn_skirtBow_R_base"(ComponentPose=@n41, Alpha=1, Translation="0.527,-0.719,2.466", Rotation="-66.042,0.000,67.824", Scale="1,1,1") {
    rotationMode=BMM_Additive
    scaleMode=BMM_Replace
    translationMode=BMM_Additive
    alphaBoolBlend {blendOption=Linear}
    alphaScaleBiasClamp {bInterpResult=true}
    boneToModify {boneName=dyn_skirtBow_R_base}
    rotation {pitch=-66.042, roll=67.824}
    translation=<0.527 -0.719 2.466>
   }
   n43 = "Transform (Modify) Bone - Bone: dyn_skirtBow_L_base"(ComponentPose=@n42, Alpha=1, Translation="0,0,0", Rotation="0.000,0.000,-20.000", Scale="1,1,1") {
    rotationMode=BMM_Additive
    scaleMode=BMM_Replace
    translationMode=BMM_Additive
    alphaBoolBlend {blendOption=Linear}
    alphaScaleBiasClamp {bInterpResult=true}
    boneToModify {boneName=dyn_skirtBow_L_base}
    rotation {roll=-20}
   }
   n44 = "Transform (Modify) Bone - Bone: dyn_skirtBow_R_center"(ComponentPose=@n43, Alpha=1, Translation="-0.000,0.000,0.681", Rotation="0.000,0.000,-10.000", Scale="1,1,1") {
    rotationMode=BMM_Additive
    scaleMode=BMM_Replace
    translationMode=BMM_Additive
    alphaBoolBlend {blendOption=Linear}
    alphaScaleBiasClamp {bInterpResult=true}
    boneToModify {boneName=dyn_skirtBow_R_center}
    rotation {roll=-10}
    translation {z=0.681}
   }
   n45 = "Transform (Modify) Bone - Bone: dyn_skirtBow_R_loop_r"(ComponentPose=@n44, Alpha=1, Translation="-0.405,6.677,7.758", Rotation="0.000,-90.000,-10.000", Scale="1,1,1") {
    rotationMode=BMM_Additive
    scaleMode=BMM_Replace
    translationMode=BMM_Additive
    alphaBoolBlend {blendOption=Linear}
    alphaScaleBiasClamp {bInterpResult=true}
    boneToModify {boneName=dyn_skirtBow_R_loop_r}
    rotation=<rot 0 -90 -10>
    translation=<-0.405 6.677 7.758>
   }
   n46 = "Transform (Modify) Bone - Bone: dyn_skirtBow_R_loop_r"(ComponentPose=@n45, Alpha=1, Translation="0,0,5", Rotation="0,0,-20", Scale="1,1,1") {
    rotationMode=BMM_Additive
    scaleMode=BMM_Replace
    translationMode=BMM_Additive
    alphaBoolBlend {blendOption=Linear}
    alphaScaleBiasClamp {bInterpResult=true}
    boneToModify {boneName=dyn_skirtBow_R_loop_r}
    rotation {roll=-20}
    translation {z=5}
   }
   n47 = "Transform (Modify) Bone - Bone: dyn_skitBow_R_tail_l_01"(ComponentPose=@n46, Alpha=1, Translation="0,0,5", Rotation="0.000,0.000,-10.000", Scale="1,1,1") {
    rotationMode=BMM_Additive
    scaleMode=BMM_Replace
    translationMode=BMM_Additive
    alphaBoolBlend {blendOption=Linear}
    alphaScaleBiasClamp {bInterpResult=true}
    boneToModify {boneName=dyn_skitBow_R_tail_l_01}
    rotation {roll=-10}
    translation {z=5}
   }
   n48 = "Transform (Modify) Bone - Bone: dyn_skirtBow_R_tail_r_01"(ComponentPose=@n47, Alpha=1, Translation="0,0,5", Rotation="0,0,0", Scale="1,1,1") {
    rotationMode=BMM_Additive
    scaleMode=BMM_Replace
    translationMode=BMM_Additive
    alphaBoolBlend {blendOption=Linear}
    alphaScaleBiasClamp {bInterpResult=true}
    boneToModify {boneName=dyn_skirtBow_R_tail_r_01}
    translation {z=5}
   }
   n49 = "Transform (Modify) Bone - Bone: dyn_skirtBow_L_base"(ComponentPose=@n48, Alpha=1, Translation="0,0,0", Rotation="0,0,0", Scale="1,1,1") {
    rotationMode=BMM_Additive
    scaleMode=BMM_Replace
    translationMode=BMM_Additive
    alphaBoolBlend {blendOption=Linear}
    alphaScaleBiasClamp {bInterpResult=true}
    boneToModify {boneName=dyn_skirtBow_L_base}
   }
   n50 = "Transform (Modify) Bone - Bone: dyn_skirtBow_L_loop_r"(ComponentPose=@n49, Alpha=1, Translation="0,0,0", Rotation="0,0,0", Scale="1,1,1") {
    rotationMode=BMM_Additive
    scaleMode=BMM_Replace
    translationMode=BMM_Additive
    alphaBoolBlend {blendOption=Linear}
    alphaScaleBiasClamp {bInterpResult=true}
    boneToModify {boneName=dyn_skirtBow_L_loop_r}
   }
   n51 = "Transform (Modify) Bone - Bone: dyn_skirtBow_L_tail_r_01"(ComponentPose=@n50, Alpha=1, Translation="0,0,0", Rotation="0,0,0", Scale="1,1,1") {
    rotationMode=BMM_Additive
    scaleMode=BMM_Replace
    translationMode=BMM_Additive
    alphaBoolBlend {blendOption=Linear}
    alphaScaleBiasClamp {bInterpResult=true}
    boneToModify {boneName=dyn_skirtBow_L_tail_r_01}
   }
   n52 = "Transform (Modify) Bone - Bone: dyn_skirtBow_L_loopTail_l_01"(ComponentPose=@n51, Alpha=1, Translation="0,0,0", Rotation="0,0,0", Scale="1,1,1") {
    rotationMode=BMM_Additive
    scaleMode=BMM_Replace
    translationMode=BMM_Additive
    alphaBoolBlend {blendOption=Linear}
    alphaScaleBiasClamp {bInterpResult=true}
    boneToModify {boneName=dyn_skirtBow_L_loopTail_l_01}
   }
   n53 = RigidBody(ComponentPose=@n52, OverrideWorldGravity="0,0,-800") {
    bClampLinearTranslationLimitToRefPose=true
    bOverrideWorldGravity=true
    overridePhysicsAsset=&/F_MED_Street_Goth_Summer_SkirtRibbon_Physics
    simulationSpace=BaseBoneSpace
    alphaBoolBlend {blendOption=Linear}
    alphaScaleBiasClamp {bInterpResult=true}
    baseBoneRef {boneName=pelvis}
    componentAppliedLinearAccClamp=<3000 3000 3000>
    componentLinearAccScale=<0.05 0.05 0.05>
    componentLinearVelScale=<0.2 0.2 0.2>
    overrideWorldGravity {z=-800}
   }
   n54 = Anim Dynamics - Bone: dyn_necklace_01!(ComponentPose=@n53, Alpha=1, GravityOverride:Vector="0,0,-350", ExternalForce="0,0,0") -> (Pose) {
    angularDampingOverride=0.8
    angularSpringConstant=75
    bAngularSpring=true
    bGravityOverrideInSimSpace=true
    bUseGravityOverride=true
    linearDampingOverride=0.9
    linearSpringConstant=150
    numSolverIterationsPreUpdate=3
    alphaBoolBlend {blendOption=Linear}
    boundBone {boneName=dyn_necklace_01}
    componentAppliedLinearAccClamp=<10000 10000 10000>
    componentLinearVelScale=<0.1 0.1 0.1>
    gravityOverride {z=-350}
    physicsBodyDefinitions [{
      boundBone {boneName=dyn_necklace_01}
      boxExtents=<12 12 12>
      constraintSetup {angularLimitsMax=<0 0 5>, angularLimitsMin=<0 -30 -5>, angularTarget=<1 0 0>}
      localJointOffset=<2 0 0>
     }]
    relativeSpaceBone {boneName=spine_05}
   }
   n55 = Anim Dynamics - Bone: dyn_necklace_02!(ComponentPose=@n54, Alpha=1, GravityOverride:Vector="0,0,-200", ExternalForce="0,0,0") -> (Pose) {
    angularDampingOverride=0.8
    angularSpringConstant=75
    bAngularSpring=true
    bGravityOverrideInSimSpace=true
    bUseGravityOverride=true
    linearDampingOverride=0.9
    linearSpringConstant=150
    numSolverIterationsPreUpdate=3
    alphaBoolBlend {blendOption=Linear}
    boundBone {boneName=dyn_necklace_02}
    componentAppliedLinearAccClamp=<10000 10000 10000>
    componentLinearVelScale=<0.1 0.1 0.1>
    gravityOverride {z=-200}
    physicsBodyDefinitions [{
      boundBone {boneName=dyn_necklace_02}
      boxExtents=<12 12 12>
      constraintSetup {angularLimitsMax=<0 2 5>, angularLimitsMin=<0 -30 -5>, angularTarget=<1 0 0>}
      localJointOffset=<1 0 0>
     }]
    relativeSpaceBone {boneName=spine_05}
   }
   n56 = Anim Dynamics - Bone: dyn_necklace_03!(ComponentPose=@n55, Alpha=1, GravityOverride:Vector="0,0,-200", ExternalForce="0,0,0") -> (Pose) {
    angularDampingOverride=0.8
    angularSpringConstant=75
    bAngularSpring=true
    bGravityOverrideInSimSpace=true
    bUseGravityOverride=true
    linearDampingOverride=0.9
    linearSpringConstant=150
    numSolverIterationsPreUpdate=3
    alphaBoolBlend {blendOption=Linear}
    boundBone {boneName=dyn_necklace_03}
    componentAppliedLinearAccClamp=<10000 10000 10000>
    componentLinearVelScale=<0.1 0.1 0.1>
    gravityOverride {z=-200}
    physicsBodyDefinitions [{
      boundBone {boneName=dyn_necklace_03}
      boxExtents=<12 12 12>
      constraintSetup {angularLimitsMax=<0 2 5>, angularLimitsMin=<0 -30 -5>, angularTarget=<1 0 0>}
      localJointOffset=<1 0 0>
     }]
    relativeSpaceBone {boneName=spine_05}
   }
   n57 = "Constraint - dyn_necklace_01 (P)"(ComponentPose=@n56, Alpha=1, ConstraintWeights_0=1) {
    alphaBoolBlend {blendOption=Linear}
    boneToModify {boneName=dyn_necklace_01}
    constraintSetup [{targetBone {boneName=spine_04}}]
    constraintWeights [1]
   }
   n58 = "Transform (Modify) Bone - Bone: dyn_necklace_01"(ComponentPose=@n57, Alpha=1, Translation="0,1,0", Rotation="5,0,0", Scale="1,1,1") {
    rotationMode=BMM_Additive
    scaleMode=BMM_Replace
    translationMode=BMM_Additive
    alphaBoolBlend {blendOption=Linear}
    alphaScaleBiasClamp {bInterpResult=true}
    boneToModify {boneName=dyn_necklace_01}
    rotation {pitch=5}
    translation {y=1}
   }
   n59 = "Transform (Modify) Bone - Bone: dyn_neckBow_base"(ComponentPose=@n58, Alpha=1, Translation="0.000,-0.632,0.000", Rotation="-0.000,-10.000,10.000", Scale="1,1,1") {
    rotationMode=BMM_Additive
    scaleMode=BMM_Replace
    translationMode=BMM_Additive
    alphaBoolBlend {blendOption=Linear}
    alphaScaleBiasClamp {bInterpResult=true}
    boneToModify {boneName=dyn_neckBow_base}
    rotation=<rot 0 -10 10>
    translation {y=-0.632}
   }
   n60 = Anim Dynamics - Bone: dyn_neckBow_center(ComponentPose=@n59, Alpha=1, ExternalForce="0,0,0") {
    angularDampingOverride=0.8
    angularSpringConstant=50
    bAngularSpring=true
    bGravityOverrideInSimSpace=true
    bUseGravityOverride=true
    linearDampingOverride=0.9
    linearSpringConstant=150
    numSolverIterationsPreUpdate=3
    alphaBoolBlend {blendOption=Linear}
    boundBone {boneName=dyn_neckBow_center}
    componentAppliedLinearAccClamp=<10000 10000 10000>
    componentLinearAccScale {z=0.2}
    componentLinearVelScale=<0.1 0.1 0.1>
    gravityOverride {z=-500}
    physicsBodyDefinitions [{
      boundBone {boneName=dyn_neckBow_center}
      boxExtents=<8 8 8>
      constraintSetup {angularLimitsMax=<0 0 5>, angularLimitsMin=<0 -100 -5>, angularTarget=<1 0 0>}
      localJointOffset=<0.5 0 0>
     }]
    relativeSpaceBone {boneName=spine_05}
   }
   n61 = Anim Dynamics - Bone: dyn_neckBow_loop_short_01(ComponentPose=@n60, Alpha=1, ExternalForce="0,0,0") {
    angularDampingOverride=0.8
    angularSpringConstant=50
    bAngularSpring=true
    bGravityOverrideInSimSpace=true
    bUseGravityOverride=true
    linearDampingOverride=0.9
    linearSpringConstant=150
    numSolverIterationsPreUpdate=3
    alphaBoolBlend {blendOption=Linear}
    boundBone {boneName=dyn_neckBow_loop_short_01}
    componentAppliedLinearAccClamp=<10000 10000 10000>
    componentLinearAccScale {z=0.2}
    componentLinearVelScale=<0.1 0.1 0.1>
    gravityOverride {z=-500}
    physicsBodyDefinitions [{
      boundBone {boneName=dyn_neckBow_loop_short_01}
      boxExtents=<10 10 10>
      constraintSetup {angularLimitsMax=<0 0 5>, angularLimitsMin=<0 -100 -5>, angularTarget=<1 0 0>}
      localJointOffset=<1 0 0>
     }]
    relativeSpaceBone {boneName=spine_05}
   }
   n62 = Anim Dynamics - Bone: dyn_neckBow_loop_long_01(ComponentPose=@n61, Alpha=1, ExternalForce="0,0,0") {
    angularDampingOverride=0.8
    angularSpringConstant=50
    bAngularSpring=true
    bGravityOverrideInSimSpace=true
    bUseGravityOverride=true
    linearDampingOverride=0.9
    linearSpringConstant=150
    numSolverIterationsPreUpdate=3
    alphaBoolBlend {blendOption=Linear}
    boundBone {boneName=dyn_neckBow_loop_long_01}
    componentAppliedLinearAccClamp=<10000 10000 10000>
    componentLinearAccScale {z=0.2}
    componentLinearVelScale=<0.1 0.1 0.1>
    gravityOverride {z=-500}
    physicsBodyDefinitions [{
      boundBone {boneName=dyn_neckBow_loop_long_01}
      boxExtents=<10 10 10>
      constraintSetup {angularLimitsMax=<0 0 5>, angularLimitsMin=<0 -100 -5>, angularTarget=<1 0 0>}
      localJointOffset=<2 0 0>
     }]
    relativeSpaceBone {boneName=spine_05}
   }
   n63 = Anim Dynamics - Bone: dyn_neckBow_tail_long_01(ComponentPose=@n62, Alpha=1, ExternalForce="0,0,0") {
    angularDampingOverride=0.8
    angularSpringConstant=75
    bAngularSpring=true
    bGravityOverrideInSimSpace=true
    bUseGravityOverride=true
    linearDampingOverride=0.9
    linearSpringConstant=150
    numSolverIterationsPreUpdate=3
    alphaBoolBlend {blendOption=Linear}
    boundBone {boneName=dyn_neckBow_tail_long_01}
    componentAppliedLinearAccClamp=<10000 10000 10000>
    componentLinearAccScale {z=0.2}
    componentLinearVelScale=<0.1 0.1 0.1>
    gravityOverride {z=-500}
    physicsBodyDefinitions [{
      boundBone {boneName=dyn_neckBow_tail_long_01}
      boxExtents=<10 10 10>
      constraintSetup {angularLimitsMax=<0 0 5>, angularLimitsMin=<0 -100 -5>, angularTarget=<1 0 0>}
      localJointOffset=<1 0 0>
     }]
    relativeSpaceBone {boneName=spine_05}
   }
   n64 = Anim Dynamics - Bone: dyn_neckBow_loop_long_02(ComponentPose=@n63, Alpha=1, ExternalForce="0,0,0") {
    angularDampingOverride=0.8
    angularSpringConstant=50
    bAngularSpring=true
    bGravityOverrideInSimSpace=true
    bUseGravityOverride=true
    linearDampingOverride=0.9
    linearSpringConstant=150
    numSolverIterationsPreUpdate=3
    alphaBoolBlend {blendOption=Linear}
    boundBone {boneName=dyn_neckBow_loop_long_02}
    componentAppliedLinearAccClamp=<10000 10000 10000>
    componentLinearAccScale {z=0.2}
    componentLinearVelScale=<0.1 0.1 0.1>
    gravityOverride {z=-500}
    physicsBodyDefinitions [{
      boundBone {boneName=dyn_neckBow_loop_long_02}
      boxExtents=<8 8 8>
      constraintSetup {angularLimitsMax=<0 0 5>, angularLimitsMin=<0 -100 -5>, angularTarget=<1 0 0>}
      localJointOffset=<1 0 0>
     }]
    relativeSpaceBone {boneName=spine_05}
   }
   n65 = Anim Dynamics - Bone: dyn_neckBow_tail_long_02(ComponentPose=@n64, Alpha=1, ExternalForce="0,0,0") {
    angularDampingOverride=0.8
    angularSpringConstant=75
    bAngularSpring=true
    bGravityOverrideInSimSpace=true
    bUseGravityOverride=true
    linearDampingOverride=0.9
    linearSpringConstant=150
    numSolverIterationsPreUpdate=3
    alphaBoolBlend {blendOption=Linear}
    boundBone {boneName=dyn_neckBow_tail_long_02}
    componentAppliedLinearAccClamp=<10000 10000 10000>
    componentLinearAccScale {z=0.3}
    componentLinearVelScale=<0.1 0.1 0.1>
    gravityOverride {z=-500}
    physicsBodyDefinitions [{
      boundBone {boneName=dyn_neckBow_tail_long_02}
      boxExtents=<8 8 8>
      constraintSetup {angularLimitsMax=<0 0 5>, angularLimitsMin=<0 -100 -5>, angularTarget=<1 0 0>}
      localJointOffset=<1 0 0>
     }]
    relativeSpaceBone {boneName=spine_05}
   }
   n66 = Anim Dynamics - Bone: dyn_neckBow_tail_short_01(ComponentPose=@n65, Alpha=1, ExternalForce="0,0,0") {
    angularDampingOverride=0.8
    angularSpringConstant=75
    bAngularSpring=true
    bGravityOverrideInSimSpace=true
    bUseGravityOverride=true
    linearDampingOverride=0.9
    linearSpringConstant=150
    numSolverIterationsPreUpdate=3
    alphaBoolBlend {blendOption=Linear}
    boundBone {boneName=dyn_neckBow_tail_short_01}
    componentAppliedLinearAccClamp=<10000 10000 10000>
    componentLinearAccScale {z=0.2}
    componentLinearVelScale=<0.1 0.1 0.1>
    gravityOverride {z=-500}
    physicsBodyDefinitions [{
      boundBone {boneName=dyn_neckBow_tail_short_01}
      boxExtents=<10 10 10>
      constraintSetup {angularLimitsMax=<0 0 5>, angularLimitsMin=<0 -100 -5>, angularTarget=<1 0 0>}
      localJointOffset=<1 0 0>
     }]
    relativeSpaceBone {boneName=spine_05}
   }
   n67 = Anim Dynamics - Bone: dyn_neckBow_tail_long_03(ComponentPose=@n66, Alpha=1, ExternalForce="0,0,0") {
    angularDampingOverride=0.8
    angularSpringConstant=75
    bAngularSpring=true
    bGravityOverrideInSimSpace=true
    bUseGravityOverride=true
    linearDampingOverride=0.9
    linearSpringConstant=150
    numSolverIterationsPreUpdate=3
    alphaBoolBlend {blendOption=Linear}
    boundBone {boneName=dyn_neckBow_tail_long_03}
    componentAppliedLinearAccClamp=<10000 10000 10000>
    componentLinearAccScale {z=0.3}
    componentLinearVelScale=<0.1 0.1 0.1>
    gravityOverride {z=-500}
    physicsBodyDefinitions [{
      boundBone {boneName=dyn_neckBow_tail_long_03}
      boxExtents=<8 8 8>
      constraintSetup {angularLimitsMax=<0 0 5>, angularLimitsMin=<0 -100 -5>, angularTarget=<1 0 0>}
      localJointOffset=<1 0 0>
     }]
    relativeSpaceBone {boneName=spine_05}
   }
   n68 = Anim Dynamics - Bone: dyn_neckBow_tail_short_02(ComponentPose=@n67, Alpha=1, ExternalForce="0,0,0") {
    angularDampingOverride=0.8
    angularSpringConstant=75
    bAngularSpring=true
    bGravityOverrideInSimSpace=true
    bUseGravityOverride=true
    linearDampingOverride=0.9
    linearSpringConstant=150
    numSolverIterationsPreUpdate=3
    alphaBoolBlend {blendOption=Linear}
    boundBone {boneName=dyn_neckBow_tail_short_02}
    componentAppliedLinearAccClamp=<10000 10000 10000>
    componentLinearAccScale {z=0.3}
    componentLinearVelScale=<0.1 0.1 0.1>
    gravityOverride {z=-500}
    physicsBodyDefinitions [{
      boundBone {boneName=dyn_neckBow_tail_short_02}
      boxExtents=<8 8 8>
      constraintSetup {angularLimitsMax=<0 0 5>, angularLimitsMin=<0 -100 -5>, angularTarget=<1 0 0>}
      localJointOffset=<1 0 0>
     }]
    relativeSpaceBone {boneName=spine_05}
   }
   n69 = "Transform (Modify) Bone - Bone: dyn_neckBow_center"(ComponentPose=@n68, Alpha=1, Translation="0.275,-0.869,-0.000", Rotation="-10.000,-10.000,20.000", Scale="1,1,1") {
    rotationMode=BMM_Additive
    scaleMode=BMM_Replace
    translationMode=BMM_Additive
    alphaBoolBlend {blendOption=Linear}
    alphaScaleBiasClamp {bInterpResult=true, interpSpeedDecreasing=0.5, interpSpeedIncreasing=0.5}
    boneToModify {boneName=dyn_neckBow_center}
    rotation=<rot -10 -10 20>
    translation=<0.275 -0.869>
   }
   Output Pose(Result=Component To Local(ComponentPose=@n69)) {layerGroup=None}
   Reroute Node(InputPin:Wildcard) -> (OutputPin:Wildcard)
  ]}
 Calculate Cloth Sim Noise {nodes [
   n1 = Calculate Cloth Sim Noise!() -> (
    then
    MID:Material Instance Dynamic Object Reference
    OverlayMID:Material Instance Dynamic Object Reference
    SkelMesh:Skeletal Mesh Component Object Reference
    BoneName:Name=root
    DeltaTime:double=~
    NoiseMovementStrength:double=0.25
    NoiseBoneLocation:"Vector (by ref)"
    NoiseVelocityVector:"Vector (by ref)"
   ) {is=title K2Node_FunctionEntry}
   > n2 = calculateMotionVectorForSocket!(
    execute
    self
    skelMesh:Scene Component Object Reference=@n1.SkelMesh
    boneName:Name=@n1.BoneName
    previousBoneLocation:Vector=@n1.NoiseBoneLocation
   ) -> (then, boneMotionVector:Vector, boneLocation:Vector, boneTransform:Transform)
   n3 = vector + vector(
    A=vector * vector(A=@n2, B=float * float(A=@n1.DeltaTime, B=@n1.NoiseMovementStrength) {is=title K2Node_PromotableOperator}) -> (ReturnValue=~)
    B=@n1.NoiseVelocityVector
   )
   > SetVectorParameterValue(
    self=[@n1.OverlayMID, @n1]
    ParameterName=ClothSim_ABP_TurbulenceUVOffset
    Value="To LinearColor (Vector)"!(InVec:Vector=@n3) -> (ReturnValue:Linear Color Structure)
   )
   > "Set Vector (by ref)"!(execute:Exec, Target:"Vector (by ref)"=@n1.NoiseBoneLocation, Value:Vector=@n2.boneLocation) -> (then:Exec) {is=title K2Node_VariableSetRef}
   > "Set Vector (by ref)"(Target=@n1.NoiseVelocityVector, Value=@n3)
   > Return Node()
  ]}
 EventGraph {nodes [
   Event BlueprintUpdateAnimation()
   > n1 = Gravity Override State Function DEPRECATED(CharacterPartAnimInstance=Self-Reference(), GameTimeInSeconds=GetGameTimeInSeconds(), GravityOverrideParams=Get Gravity Override Params())
   > Set Gravity Override(Gravity Override=@n1)
   > n2 = Branch(Condition=AND Boolean(A=Get bIsCrouching(), B=Get bIsMoving2D()))
   > n3 = Set crouch walking(crouch walking:double=1) -> (Output_Get:double)
   n4 = Branch(execute=[@n3.then, @n46.then], Condition=Get bIsSkyDiving())
   > Set skirtAlpha(skirtAlpha:double=0.8) -> (Output_Get:double)
   > n5 = Set is jumping/falling(is jumping/falling:double=0) -> (Output_Get:double)
   n6 = "To Float (Boolean)"(InBool=Get bIsCrouching())
   n7 = GetOwnerSkeletalMeshComponent()
   n8 = BreakRotator(InRot=BreakTransform(InTransform=GetDeltaTransformFromRefPose(self=@n7, BoneName=thigh_r, BaseName=pelvis)).Rotation)
   Set Crouch R Thigh Rot In(
    execute=[@n5.then, @n44.then, @n45.then]
    Crouch R Thigh Rot In:double=float * float(A=MapRangeClamped(Value=@n8.Pitch, InRangeA=1, InRangeB=-12), B=@n6)
   ) -> (Output_Get:double)
   > Set Crouch R Thigh Rot Out(Crouch R Thigh Rot Out:double=float * float(A=MapRangeClamped(Value=@n8.Pitch, InRangeA=20, InRangeB=37), B=@n6)) -> (Output_Get:double)
   n9 = MapRangeClamped(Value="Absolute (Float)"(A=@n8.Yaw), InRangeA=0, InRangeB=110)
   > Set R Thigh Rot Up(R Thigh Rot Up:double=@n9) -> (Output_Get:double)
   > Set Crouch R Thigh Rot Up(Crouch R Thigh Rot Up:double=float * float(A=@n9, B=@n6)) -> (Output_Get:double)
   n10 = GetOwnerSkeletalMeshComponent()
   n11 = BreakRotator(InRot=BreakTransform(InTransform=GetDeltaTransformFromRefPose(self=@n10, BoneName=thigh_l, BaseName=pelvis)).Rotation)
   > Set Crouch L Thigh Rot In(Crouch L Thigh Rot In:double=float * float(A=@n6, B=MapRangeClamped(Value=@n11.Pitch, InRangeA=1, InRangeB=21))) -> (Output_Get:double)
   n12 = MapRangeClamped(Value="Absolute (Float)"(A=@n11.Yaw), InRangeA=20, InRangeB=110)
   > Set L Thigh Rot Up(L Thigh Rot Up:double=@n12) -> (Output_Get:double)
   > Set L Thigh Rot Out(L Thigh Rot Out:double=MapRangeClamped(Value=@n11.Pitch, InRangeA=25, InRangeB=41)) -> (Output_Get:double)
   > Set Crouch L Thigh Rot Up(Crouch L Thigh Rot Up:double=float * float(A=@n6, B=@n12)) -> (Output_Get:double)
   > Set L Thigh Rot Back(L Thigh Rot Back:double=MapRangeClamped(Value=@n11.Yaw, InRangeA=0, InRangeB=7)) -> (Output_Get:double)
   n13 = Get bIsTargeting()
   n14 = GetOwnerSkeletalMeshComponent()
   n15 = BreakRotator(InRot=BreakTransform(InTransform=GetDeltaTransformFromRefPose(self=@n14, BoneName=head, BaseName=spine_05)).Rotation)
   > Set Head Up Targeting(Head Up Targeting:double=float * float(A=MapRangeClamped(Value=@n15.Yaw, InRangeA=0, InRangeB=-17), B="To Float (Boolean)"(InBool=@n13))) -> (Output_Get:double)
   > Set Head Down(Head Down:double=MapRangeClamped(Value=@n15.Yaw, InRangeA=1, InRangeB=18)) -> (Output_Get:double)
   > n16 = Branch(Condition=Get bIsCrouching())
   > Set Ribbon RBAN Alpha(Ribbon RBAN Alpha:double=0.8) -> (Output_Get:double)
   > n17 = Set Ribbon Constraint Alpha(Ribbon Constraint Alpha:double=1) -> (Output_Get:double)
   n18 = Branch(execute=[@n42.then, @n17.then, @n41.then], Condition=CheckForBackbling())
   > n19 = Set Back Bow Alpha(Back Bow Alpha:double=0) -> (Output_Get:double)
   n20 = Get bIsPlayingMeleeAnim()
   Set Melee Swing(execute=[@n39.then, @n19.then], Melee Swing:double="To Float (Boolean)"(InBool=@n20)) -> (Output_Get:double)
   > n21 = Branch(Condition=@n20)
   > n22 = Set Back Bow Alpha(Back Bow Alpha=0.8)
   n23 = Sequence!(execute=[@n22.then, @n38.then]) -> (then_0, then_1, then_2:Exec, then_3:Exec, then_4:Exec)
   n24 = Get bIsSkyDiving()
   > n25 = Branch(Condition=@n24)
   > n26 = Set ClothSim_WindStrength(ClothSim_WindStrength:double=0.4) -> (Output_Get:double)
   n27 = Get Mat Instance Dynamic()
   SetScalarParameterValue(
    execute=[@n28.then, @n26.then]
    self=@n27
    ParameterName=ClothSim_WindStrength
    Value=Get ClothSim_WindStrength() -> (ClothSim_WindStrength:double)
   )
   n28 = Set ClothSim_WindStrength(execute=@n25.else, ClothSim_WindStrength=0)
   n29 = Branch(execute=@n23.then_1, Condition=@n24)
   > n30 = Set ClothSim_VelocityMax(ClothSim_VelocityMax:double=750) -> (Output_Get:double)
   SetScalarParameterValue(
    execute=[@n31.then, @n30.then]
    self=@n27
    ParameterName=ClothSim_VelocityMax
    Value=Get ClothSim_VelocityMax() -> (ClothSim_VelocityMax:double)
   )
   n31 = Set ClothSim_VelocityMax(execute=@n29.else, ClothSim_VelocityMax=5000)
   n32 = Branch(execute=@n23.then_2, Condition=@n24)
   > n33 = Set ClothSim_WindNoiseSpeed(ClothSim_WindNoiseSpeed:double=200) -> (Output_Get:double)
   SetScalarParameterValue(
    execute=[@n34.then, @n33.then]
    self=@n27
    ParameterName=ClothSim_WindNoiseSpeed
    Value=Get ClothSim_WindNoiseSpeed() -> (ClothSim_WindNoiseSpeed:double)
   )
   n34 = Set ClothSim_WindNoiseSpeed(execute=@n32.else, ClothSim_WindNoiseSpeed=0)
   n35 = Branch(execute=@n23.then_4, Condition=Get movementSum() -> (movementSum:bool))
   SetVectorParameterValue(
    execute=@n35.else
    self=Get Mat Instance Dynamic()
    ParameterName=ClothSim_ABP_R1_Velocity
    Value=Get Value() -> (Value:Linear Color Structure)
   )
   Set WPO Globals Function(
    execute=@n23.then_3
    Material=Get Mat Instance Dynamic()
    Overlay Material=Get Overlay Mat Instance Dynamic()
    Skeletal Mesh=Get skeletal_mesh() -> (skeletal_mesh:Skeletal Mesh Component Object Reference)
    "Local Noise Transform  Location"=Get Local Noise Transform Location() -> (Local Noise Transform Location:Vector)
    Local Noise Offset=Get Local Noise Offset()
   )
   > Set WPO Regional Function!(
    execute
    Material:Material Instance Dynamic Object Reference=Get Mat Instance Dynamic()
    Overlay Material:Material Instance Dynamic Object Reference=Get Overlay Mat Instance Dynamic()
    Skeletal Mesh:Skeletal Mesh Component Object Reference=Get skeletal_mesh()
    Velocity Transform:Name=pelvis
    Velocity Interp Speed:double=Get Velocity Interp Speed()
    Collision Transform:Name=spine_01
    Collision Transform Axis:EAxis Enum
    Invert Collision Transform Axis:bool=false
    Delta Time X:double=Reroute Node(InputPin:double) -> (OutputPin:double)
    Previous Velocity Transform Location:"Vector (by ref)"=Get R1 Velocity Transform Location()
    Previous Velocity Transform Velocity:"Vector (by ref)"=Get R1 Velocity Transform Velocity()
    Region:int=1
   ) -> (then)
   > Get!(execute:Exec) -> (then:Exec, else:Exec, ownerSkelMesh:Skeletal Mesh Component Object Reference)
   > Get!(execute) -> (then, else, PlayerPawn:Fort Player Pawn Athena Object Reference)
   > n36 = PlayerMovement(AnimInstance=Self-Reference(), XDotProductCoefficient=1, YDotProductCoefficient=1, ZDotProductCoefficient=1)
   > n37 = Branch(Condition="Not Equal (Float)"(A=float + float!(A=@n36, B=@n36.YMovement, C:float=@n36.ZMovement) -> (ReturnValue)))
   > Set movementSum(movementSum:bool=true) -> (Output_Get:bool)
   Set movementSum(execute=@n37.else, movementSum=false)
   n38 = Set Back Bow Alpha(execute=@n21.else, Back Bow Alpha=1)
   n39 = Set Back Bow Alpha(execute=@n18.else, Back Bow Alpha=1)
   Set Ribbon RBAN Alpha(execute=@n16.else, Ribbon RBAN Alpha=1)
   > Set Ribbon Constraint Alpha(Ribbon Constraint Alpha=1)
   > n40 = Branch(Condition=Get bIsPlayingEmote())
   > n41 = Set Ribbon RBAN Alpha(Ribbon RBAN Alpha=0.5)
   n42 = Set Ribbon RBAN Alpha(execute=@n40.else, Ribbon RBAN Alpha=1)
   Set skirtAlpha(execute=@n4.else, skirtAlpha=1)
   > n43 = Branch(Condition=OR Boolean(A=Get bIsJumping(), B=Get bIsFalling()))
   > n44 = Set is jumping/falling(is jumping/falling=1)
   n45 = Set is jumping/falling(execute=@n43.else, is jumping/falling=0)
   n46 = Set crouch walking(execute=@n2.else, crouch walking=0)
   Event OnCharacterCustomizationLoaded()
   > n47 = Set skeletal_mesh(skeletal_mesh:Skeletal Mesh Component Object Reference=GetPartSkeletalMeshComponent()) -> (Output_Get:Skeletal Mesh Component Object Reference)
   > n48 = Get Dynamic Material Instance(SkeletalMeshComponent=@n47)
   > Set Mat Instance Dynamic(Mat Instance Dynamic=@n48)
   > n49 = GetOrCreateDynamicOverlayMaterial(MeshComponent=@n47, MaterialName=None)
   > Set Overlay Mat Instance Dynamic(Overlay Mat Instance Dynamic=@n49)
  ]}
 GetBoneNames {nodes [
   n1 = GetBoneNames!() -> (then, Region:"Integer (by ref)") {is=title K2Node_FunctionEntry}
   n2 = Append(A=ClothSim_ABP_R, B="To String (Integer)"(InInt=@n1))
   > Return Node!(
    execute
    Velocity:Name=String To Name(InString=Append(A=@n2, B=_Velocity))
    BonePosition:Name=String To Name(InString=Append(A=@n2, B=_BonePosition))
    BoneVector:Name=String To Name(InString=Append(A=@n2, B=_BoneVector))
   )
  ]}
 Set Cloth Sim Params {nodes [
   n1 = Set Cloth Sim Params!() -> (
    then
    material:Material Instance Dynamic Object Reference
    overlayMaterial:Material Instance Dynamic Object Reference
    skelMesh:Skeletal Mesh Component Object Reference
    boneName:Name
    interpSpeed:double=2
    collisionBoneName:Name
    collisionBoneAxis:EAxis Enum=None
    invertCollisionBoneAxis:bool=~
    deltaTimeX:double=~
    previousBoneLocation:"Vector (by ref)"
    previousBoneVelocityVector:"Vector (by ref)"
    region:int=1
   ) {is=title K2Node_FunctionEntry}
   > Set lMID(lMID:Material Instance Dynamic Object Reference=@n1) -> (Output_Get:Material Instance Dynamic Object Reference)
   > Set loverlayMID(loverlayMID:Material Instance Dynamic Object Reference=@n1.overlayMaterial) -> (Output_Get:Material Instance Dynamic Object Reference)
   > Set lSkelMesh(lSkelMesh:Skeletal Mesh Component Object Reference=@n1.skelMesh) -> (Output_Get:Skeletal Mesh Component Object Reference)
   > n2 = GetBoneNames!(execute, self, Region:"Integer (by ref)"=@n1.region) -> (then, Velocity:Name=None, BonePosition:Name=None, BoneVector:Name=None) {is=title K2Node_CallFunction}
   > Set lMatVelocityParamName(lMatVelocityParamName:Name=@n2) -> (Output_Get:Name=None)
   > Set lMatBonePositionParamName(lMatBonePositionParamName:Name=@n2.BonePosition) -> (Output_Get:Name=None)
   > Set lMatBoneVectorParamName(lMatBoneVectorParamName:Name=@n2.BoneVector) -> (Output_Get:Name=None)
   > n3 = calculateMotionVectorForSocket(
    skelMesh=Get lSkelMesh() -> (lSkelMesh:Skeletal Mesh Component Object Reference)
    boneName=@n1.boneName
    previousBoneLocation=@n1.previousBoneLocation
   )
   n4 = VInterpTo!(
    Current:Vector=@n1.previousBoneVelocityVector
    Target:Vector=vector / vector(A:Vector=@n3, B=@n1.deltaTimeX) -> (ReturnValue:Vector)
    DeltaTime:float=@n1.deltaTimeX
    InterpSpeed:float=@n1.interpSpeed
   ) -> (ReturnValue:Vector)
   > SetVectorParameterValue(
    self=[
     Get loverlayMID() -> (loverlayMID:Material Instance Dynamic Object Reference)
     Get lMID() -> (lMID:Material Instance Dynamic Object Reference)
    ]
    ParameterName=Get lMatVelocityParamName() -> (lMatVelocityParamName:Name=None)
    Value="To LinearColor (Vector)"(InVec=@n4)
   )
   n5 = GetSocketTransform(self=Get lSkelMesh(), InSocketName=@n1.collisionBoneName, TransformSpace=RTS_World)
   n6 = "Get Unit Axis (Matrix)"!(
    M:"Matrix Structure (by ref)"="To Matrix (Transform)"!(Transform:"Transform (by ref)"=@n5) -> (ReturnValue:Matrix Structure)
    Axis:EAxis Enum=@n1.collisionBoneAxis
   ) -> (ReturnValue:Vector) {is=title K2Node_CallFunction}
   n7 = Select(Option 0:double=1, Option 1:double=-1, Index=@n1.invertCollisionBoneAxis) -> (ReturnValue:double)
   > SetVectorParameterValue(
    self=[Get loverlayMID(), Get lMID()]
    ParameterName=Get lMatBoneVectorParamName() -> (lMatBoneVectorParamName:Name=None)
    Value="To LinearColor (Vector)"(InVec=vector * vector(A=@n6, B=@n7) -> (ReturnValue=~))
   )
   > SetVectorParameterValue(
    self=[Get loverlayMID(), Get lMID()]
    ParameterName=Get lMatBonePositionParamName() -> (lMatBonePositionParamName:Name=None)
    Value="To LinearColor (Vector)"(InVec=BreakTransform(InTransform=@n5))
   )
   > "Set Vector (by ref)"(Target=@n1.previousBoneLocation, Value=@n3.boneLocation)
   > "Set Vector (by ref)"(Target=@n1.previousBoneVelocityVector, Value=@n4)
   > Return Node!(execute)
   Set lMID()
  ]}
 calculateMotionVectorForSocket {nodes [
   n1 = calculateMotionVectorForSocket!() -> (then, skelMesh:Scene Component Object Reference, boneName:Name, previousBoneLocation:Vector=~) {is=title K2Node_FunctionEntry}
   n2 = GetSocketTransform(self=@n1, InSocketName=@n1.boneName, TransformSpace=RTS_World)
   n3 = BreakTransform(InTransform=@n2)
   > Return Node!(execute, boneMotionVector:Vector=vector - vector(A=@n3, B=@n1.previousBoneLocation), boneLocation:Vector=@n3, boneTransform:Transform=@n2)
  ]}
 setWPOMaterialVelocity {nodes [
   n1 = setWPOMaterialVelocity!() -> (
    then
    materialInstance:Material Instance Dynamic Object Reference
    skelMesh:Skeletal Mesh Component Object Reference
    boneName:Name
    offsetBoneName:Name
    offsetBoneAxis:EAxis Enum=None
    deltaTimeX:double=~
    materialVelocityAttribute:Name=Velocity
    previousBoneLocation:Vector=~
    previousBoneVelocityVector:Vector=~
   ) {is=title K2Node_FunctionEntry}
   > n2 = calculateMotionVectorForSocket(skelMesh=@n1.skelMesh, boneName=@n1.boneName, previousBoneLocation=@n1.previousBoneLocation) {is=title K2Node_CallFunction}
   n3 = "Get Unit Axis (Matrix)"(
    M="To Matrix (Transform)"(Transform=GetSocketTransform(self=@n1.skelMesh, InSocketName=@n1.offsetBoneName, TransformSpace=RTS_World))
    Axis=@n1.offsetBoneAxis
   )
   n4 = vector / vector(A=@n2, B=@n1.deltaTimeX)
   n5 = VInterpTo(Current=@n1.previousBoneVelocityVector, Target=@n4, DeltaTime=@n1.deltaTimeX, InterpSpeed=0.75)
   > SetVectorParameterValue(self=@n1, ParameterName=@n1.materialVelocityAttribute, Value="To LinearColor (Vector)"(InVec=vector + vector(A=@n3, B=@n5)))
   > Return Node!(
    execute
    boneVelocity:double=Vector Length(A=@n4)
    boneVelocityVector:Vector=@n4
    boneLocation=@n2.boneLocation
    boneTransform=@n2.boneTransform
   )
  ]}
]
variables [
 Back Bow Alpha {type=number}
 ClothSim_VelocityMax {type=number}
 ClothSim_WindNoiseSpeed {type=number}
 ClothSim_WindStrength {type=number}
 Crouch L Thigh Rot In {type=number}
 Crouch L Thigh Rot Up {type=number}
 Crouch R Thigh Rot In {type=number}
 Crouch R Thigh Rot Out {type=number}
 Crouch R Thigh Rot Up {type=number}
 crouch walking {type=number}
 Gravity Override {type=Vector}
 Gravity Override Params {type=GravityOverrideParamsStruct}
 Head Down {type=number}
 Head Up Targeting {type=number}
 is jumping/falling {type=number}
 L Thigh Rot Back {type=number}
 L Thigh Rot Out {type=number}
 L Thigh Rot Up {type=number}
 Local Noise Offset {type=Vector}
 Local Noise Transform Location {type=Vector}
 Mat Instance Dynamic {type=MaterialInstanceDynamic}
 Melee Swing {type=number}
 movementSum {type=boolean}
 Overlay Mat Instance Dynamic {type=MaterialInstanceDynamic}
 ownerSkelMesh {type=SkeletalMeshComponent}
 PlayerPawn {type=FortPlayerPawnAthena}
 R Thigh Rot Up {type=number}
 R1 Velocity Transform Location {type=Vector}
 R1 Velocity Transform Velocity {type=Vector}
 Ribbon Constraint Alpha {type=number}
 Ribbon RBAN Alpha {type=number}
 skeletal_mesh {type=SkeletalMeshComponent}
 skirtAlpha {type=number}
 Value {type=LinearColor, default=<rgba 0 0 0 0>}
 Velocity Interp Speed {type=number}
]
